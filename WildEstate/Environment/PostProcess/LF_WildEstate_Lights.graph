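# WildEstate light function material that computes distance-based attenuation shaping for lights.
# World-space distance to the object position, normalized by Attenuation and inverted, is raised to Falloff to form the base curve.
# Three branches derive from it: a reciprocal, an inverted curve with a soft outer edge fade, and an inverted curve multiplied by Round-quantized distance bands.
# None of these branches is shown wired to a material output.
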Material /WildEstate/Environment/PostProcess/LF_WildEstate_Lights
graph {
 properties {bAutomaticallySetUsageInEditor=false, materialDomain=MD_LightFunction, usedShadingModels=Default Lit}
 nodes [
  n1 = Divide(A=Distance(A=WorldPosition().XYZ, B=ObjectPositionWS()), B=ScalarParameter() {defaultValue=1000, parameterName=Attenuation})
  n2 = Power(Base=OneMinus(None=@n1), Exp=ScalarParameter() {defaultValue=0.2, parameterName=Falloff})
  Divide(A=Constant() {r=1}, B=@n2)
  n3 = OneMinus(None=@n2)
  n4 = Distance(A=WorldPosition().XYZ, B=ObjectPositionWS())
  n5 = ScalarParameter() {defaultValue=16.3566, parameterName=EdgeFade}
  n6 = Subtract(A=ScalarParameter() {defaultValue=1000, parameterName=Attenuation}, B=@n5)
  n7 = Saturate(None=Divide(A=Subtract(A=@n4, B=@n6), B=@n5))
  Multiply(A=@n3, B=Power(Base=OneMinus(None=@n7), Exp=ScalarParameter() {defaultValue=5.8854, parameterName=EdgeFalloff}))
  n8 = ScalarParameter() {defaultValue=5.8203, parameterName=Steps}
  n9 = Multiply(A=@n8, B=OneMinus(None=@n1))
  Multiply(A=Divide(A=Round(None=@n9), B=@n8), B=@n3)
 ]
}
