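# Post-process material function that draws a soft, flickering glow outline around custom-stencil objects, blended over the scene by an Outline_Mask input.
# A Custom HLSL node radially samples stencil matches with distance weighting and per-sample depth occlusion.
# Stepped panning noise varies thickness and optional two-tone color, while layered sines define the flicker.
# A platform switch zeroes the outline on Android, PC-Perf and the default path, and static switches control two-tone color and interior culling.
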MaterialFunction /WildEstate/Environment/PostProcess/MF_WildEstate_Softoutline
graph {
 properties {previewBlendMode=BLEND_Translucent, previewMaterialDomain=MD_PostProcess}
 nodes [
  n1 = FunctionInput(Preview=ComponentMask(None=SceneTexture() {sceneTextureId=PPI_PostProcessInput0}.Color) {b=true, g=true, r=true}) {bUsePreviewValueAsDefault=true}
  n2 = ComponentMask(None=SceneTexture() {sceneTextureId=PPI_PostProcessInput0}.Color) {b=true, g=true, r=true}
  n3 = MaterialFunctionCall!(ID=ScalarParameter() {defaultValue=5, group=WildEstate_Softoutline, parameterName=StencilIndex, sortPriority=31}) {materialFunction=/Game/Packages/Fortress_SharedMaterials/Stencil_ID_Definitions/StencilMaskCompare}
  n4 = VectorParameter() {group=WildEstate_Softoutline, parameterName=Color, defaultValue=<rgba 0.4164 2 0.0873 1>}
  n5 = TextureCoordinate()
  n6 = ScalarParameter() {defaultValue=2.4394, group=WildEstate_Softoutline_Noise, parameterName=tile noise, sortPriority=33}
  n7 = Time()
  n8 = ScalarParameter() {defaultValue=4, group=WildEstate_Softoutline_Noise, parameterName=Speed noise, sortPriority=33}
  n9 = ScalarParameter() {defaultValue=1.5, group=WildEstate_Softoutline_Noise, parameterName=Texture FPS noise, sortPriority=33}
  n10 = Floor(None=Multiply(A=Multiply(A=@n7, B=@n8), B=@n9))
  n11 = ScalarParameter() {defaultValue=0.1, group=WildEstate_Softoutline_Noise, parameterName=X speed noise, sortPriority=33}
  n12 = ScalarParameter() {defaultValue=0.3, group=WildEstate_Softoutline_Noise, parameterName=Y speed noise, sortPriority=33}
  n13 = TextureSampleParameter2D(UVs=Panner(Coordinate=Multiply(A=@n5, B=@n6), Time=Divide(A=@n10, B=@n9), Speed=AppendVector(A=@n11, B=@n12)) {bFractionalPart=true, speedX=0.02, speedY=1}) {group=WildEstate_Softoutline_Noise, parameterName=FlickerTexture, texture=/WildEstate/Environment/Textures/Water_01_Pack}
  n14 = ScalarParameter() {defaultValue=1, group=WildEstate_Softoutline_Noise, parameterName=Noise_Multi}
  n15 = Saturate(None=Power(Base=Multiply(A=@n13.RGB, B=@n14), Exp=ScalarParameter() {defaultValue=1, group=WildEstate_Softoutline_Noise, parameterName=Noise_Pow}))
  n16 = StaticSwitchParameter(
   True=LinearInterpolate(A=VectorParameter() {group=WildEstate_Softoutline, parameterName=ColorB, defaultValue=<rgba 0 2 0.5019 1>}.RGB, B=@n4.RGB, Alpha=@n15)
   False=@n4.RGB
  ) {group=WildEstate_Softoutline, parameterName=MultiTone_Color}
  n17 = LinearInterpolate(
   A=@n16
   B=Multiply(A=@n16, B=ScalarParameter() {defaultValue=0.5, group=WildEstate_Softoutline, parameterName=inner_strength})
   Alpha=@n3.Result
  )
  n18 = Multiply(A=@n17, B=ScalarParameter() {defaultValue=2, group=WildEstate_Softoutline, parameterName=Color_Flicker_Str})
  n19 = NamedRerouteUsage()
  MaterialExpressionTextureCoordinate_0 = TextureCoordinate()
  n20 = ScalarParameter() {defaultValue=3, group=WildEstate_Softoutline, parameterName=Thickness, sortPriority=31}
  n21 = LinearInterpolate(
   A=@n20
   B=Add(A=@n20, B=ScalarParameter() {defaultValue=1, group=WildEstate_Softoutline, parameterName=Thickness_Flicker, sortPriority=31})
   Alpha=NamedRerouteUsage()
  )
  MaterialExpressionLinearInterpolate_4 = LinearInterpolate(
   A=@n21
   B=Add(A=@n21, B=ScalarParameter() {defaultValue=8, group=WildEstate_Softoutline, parameterName=thickness offset, sortPriority=31})
   Alpha=@n15
  )
  MaterialExpressionScalarParameter_12 = ScalarParameter() {defaultValue=7, group=WildEstate_Softoutline, parameterName=DistanceSteps, sortPriority=31}
  MaterialExpressionScalarParameter_13 = ScalarParameter() {defaultValue=6, group=WildEstate_Softoutline, parameterName=RadialSteps, sortPriority=31}
  MaterialExpressionScalarParameter_14 = ScalarParameter() {defaultValue=0.62, group=WildEstate_Softoutline, parameterName=RadialOffset, sortPriority=31}
  MaterialExpressionScalarParameter_15 = ScalarParameter() {defaultValue=-1.2996, group=WildEstate_Softoutline, parameterName=KernelPower, sortPriority=31}
  MaterialExpressionScalarParameter_6 = ScalarParameter() {defaultValue=5, group=WildEstate_Softoutline, parameterName=StencilIndex, sortPriority=31}
  n22 = Custom!(
   UV=@MaterialExpressionTextureCoordinate_0
   Distance=@MaterialExpressionLinearInterpolate_4
   DistanceSteps=@MaterialExpressionScalarParameter_12
   RadialSteps=@MaterialExpressionScalarParameter_13
   RadialOffset=@MaterialExpressionScalarParameter_14
   KernelPower=@MaterialExpressionScalarParameter_15
   StencilIndex=@MaterialExpressionScalarParameter_6
  ) {
   code="float3 CurColor = 0;\r\nfloat2 NewUV = UV;\r\nint i = 0;\r\nfloat StepSize = Distance / (int) DistanceSteps;\r\nfloat CurDistance = 0;\r\nfloat2 CurOffset = 0;\r\nfloat SubOffset = 0;\r\nfloat TwoPi = 6.283185;\r\nfloat accumdist = 0;\r\n\r\nfloat2 texelSize = View.ViewSizeAndInvSize.zw * Distance;\r\n//float tmpstencil = 0;\r\n//float tmpstencil = SceneTextureLookup(ViewportUVToBufferUV(UV), 25, false);\r\n\r\nif (DistanceSteps < 1)\r\n{\r\n    \r\n\treturn 0;\r\n}\r\n \r\nelse\r\n{\r\n\twhile (i < (int) DistanceSteps)\r\n\t{\r\n\t\tCurDistance += StepSize;\r\n\t\tfor (int j = 0; j < (int) RadialSteps; j++)\r\n\t\t{\r\n\t\t\tSubOffset++;\r\n\t\t\tCurOffset.x = cos(TwoPi * (SubOffset / RadialSteps)) * texelSize.r;\r\n\t\t\tCurOffset.y = sin(TwoPi * (SubOffset / RadialSteps)) * texelSize.g;\r\n\t\t\tNewUV.x = UV.x + CurOffset.x * CurDistance;\r\n\t\t\tNewUV.y = UV.y + CurOffset.y * CurDistance;\r\n\t\t\tfloat distpow = pow(CurDistance, KernelPower);\r\n            float tmpstencil = SceneTextureLookup(ViewportUVToBufferUV(NewUV), 25, false);\r\n            float validmask = 1-floor(saturate(abs(tmpstencil-StencilIndex)+0.01));\r\n            float tmpdepth = SceneTextureLookup(ViewportUVToBufferUV(NewUV), 13, false);\r\n            float tmpscene = CalcSceneDepth(ScreenAlignedPosition(GetScreenPosition(Parameters)))+3.00000f;\r\n            float occlud = select((tmpdepth.r > tmpscene), 0.00000000f, 1.00000000f);  //occlusion may be to expensive in a loop\r\n\t\t\tCurColor += SceneTextureLookup(ViewportUVToBufferUV(NewUV), PPI_CustomStencil, false) * distpow*validmask*occlud;\r\n\t\t\taccumdist += distpow;\r\n\t\t}\r\n\t\tSubOffset += RadialOffset;\r\n\t\ti++;\r\n\t}\r\n\treturn CurColor / accumdist;\r\n}\r\n"
   showCode=true
   inputs [
    {inputName=UV, input {expression=&self:MaterialExpressionTextureCoordinate_0}}
    {inputName=Distance, input {expression=&self:MaterialExpressionLinearInterpolate_4}}
    {
     inputName=DistanceSteps
     input {expression=&self:MaterialExpressionScalarParameter_12}
    }
    {inputName=RadialSteps, input {expression=&self:MaterialExpressionScalarParameter_13}}
    {
     inputName=RadialOffset
     input {expression=&self:MaterialExpressionScalarParameter_14}
    }
    {inputName=KernelPower, input {expression=&self:MaterialExpressionScalarParameter_15}}
    {inputName=StencilIndex, input {expression=&self:MaterialExpressionScalarParameter_6}}
   ]
  }
  n23 = Saturate(None=ComponentMask(None=@n22) {r=true})
  n24 = Constant()
  n25 = MaterialFunctionCall("Xb1,PS4"=@n23, "PC-High&&Epic,XSX,PS5"=@n23, "PC-Perf,Android"=@n24, Switch=@n24, "PC-Low(NotPerf)&&Medium"=@n23) {materialFunction=&PlatformSwitch_1D}
  n26 = Add(
   A=StaticSwitchParameter(True=LinearInterpolate(A=@n2, Alpha=@n3.Result) {constB=0}, False=@n2) {group=WildEstate_Softoutline, parameterName=Cull_Interior}
   B=Multiply(A=LinearInterpolate(A=@n17, B=@n18, Alpha=@n19), B=@n25.Result)
  )
  n27 = Constant()
  FunctionOutput(None=LinearInterpolate(A=@n1, B=LinearInterpolate(A=@n1, B=@n26, Alpha=@n25.Result), Alpha=FunctionInput(Preview=@n27) {inputName=Outline_Mask}))
  MaterialFunctionCall() {materialFunction=&StencilMaskCompare}
  n28 = SceneTexture() {sceneTextureId=PPI_CustomDepth}
  n29 = SceneDepth()
  NamedRerouteDeclaration(None=If(A=ComponentMask(None=@n28.Color) {r=true}, B=Add(A=@n29.R) {constB=3}, "A > B"=Constant(), "A < B"=Constant() {r=1})) {name=occlusion, nodeColor=<rgba 0 1 0.3906 1>}
  n30 = ScalarParameter() {group=WildEstate_Softoutline_Flicker, parameterName=FlickerAdd}
  n31 = Time()
  n32 = Multiply(A=ScalarParameter() {defaultValue=1, group=WildEstate_Softoutline_Flicker, parameterName=FlickerSpeed}, B=@n31)
  n33 = AppendVector(A=AppendVector(A=Sine(None=@n32) {period=1.6}, B=Sine(None=@n32) {period=1.42}), B=Sine(None=@n32) {period=1.51})
  n34 = ScalarParameter() {defaultValue=1, group=WildEstate_Softoutline_Flicker, parameterName=FlickerFrequency}
  n35 = Add(A=Sine(None=@n32) {period=2.1}, B=Multiply(B=@n34) {constA=0.25})
  n36 = Add(A=Sine(None=@n32) {period=0.72}, B=Multiply(B=@n34) {constA=0.75}) {constB=0.75}
  n37 = Ceil(None=Add(A=Multiply(B=@n34) {constA=0.9, constB=0.9}, B=Sine(None=@n32) {period=1.25}) {constB=0.9})
  n38 = Saturate(None=Multiply(A=Multiply(A=@n35, B=@n36), B=Multiply(A=@n36, B=@n37)))
  n39 = LinearInterpolate(Alpha=ConstantBiasScale(None=Sine(None=@n31) {period=0.075})) {constA=0.7}
  n40 = Multiply(
   A=Multiply(A=Ceil(None=@n33), B=Multiply(A=@n38, B=@n39))
   B=ScalarParameter() {defaultValue=1, group=WildEstate_Softoutline_Flicker, parameterName=FlickerMult}
  )
  NamedRerouteDeclaration(None=Saturate(None=Add(A=@n30, B=@n40))) {name=flicker_Speed, nodeColor=<rgba 1 0.5391 0 1>}
  ScreenPosition()
  Subtract(A=ComponentMask(None=SceneTexture() {sceneTextureId=PPI_CustomDepth}.Color) {r=true}) {constB=10000}
 ]
}
