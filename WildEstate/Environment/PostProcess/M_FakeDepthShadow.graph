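# Unlit modulate material that fakes a soft shadow by darkening scene depth inside a scaled ellipsoid, drawn without depth testing.
# Scene position rebuilt from SceneDepth and camera-relative world position drives the radial falloff, blended toward no darkening outside.
# CustomDepth comparisons exempt custom-depth pixels from darkening within a tunable depth range.
# A static switch shifts mesh and falloff using PlayerPosition; mobile outputs constants, disabling the effect.
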Material /WildEstate/Environment/PostProcess/M_FakeDepthShadow
graph {
 properties {bAutomaticallySetUsageInEditor=false, bDisableDepthTest=true, blendMode=BLEND_Modulate, shadingModel=MSM_Unlit, usedShadingModels=Unlit}
 nodes [
  Constant()
  n1 = CollectionParameter() {collection=&FortniteMaterialParameters, parameterName=PlayerPosition}
  n2 = ComponentMask(None=@n1) {b=true, g=true, r=true}
  n3 = ObjectPositionWS()
  n4 = Constant3Vector()
  ShadingPathSwitch(Default=StaticSwitchParameter(True=Subtract(A=@n2, B=@n3), False=@n4) {defaultValue=true, parameterName=Enable Player Attach}, Mobile=@n4)
  n5 = ComponentMask(None=SceneTexture() {sceneTextureId=PPI_CustomDepth}.Color) {r=true}
  n6 = Subtract(A=Add(A=SceneDepth().R), B=@n5)
  n7 = Subtract(A=@n5, B=PixelDepth().R)
  n8 = ScalarParameter() {defaultValue=170, parameterName=CustomDepthMask}
  n9 = Saturate(None=Divide(A=Subtract(A=@n7, B=@n8) {constB=5000}, B=ScalarParameter() {defaultValue=20, parameterName=CustomDepthMaskLegnth}))
  n10 = Add(A=Multiply(A=Saturate(None=@n6), B=OneMinus(None=@n9)), B=ScalarParameter() {defaultValue=0.65, parameterName=Minimum Opacity})
  n11 = SceneDepth()
  n12 = PixelDepth()
  n13 = Subtract(
   A=Multiply(A=Divide(A=@n11.R, B=@n12.R), B=WorldPosition() {worldPositionShaderOffset=WPT_CameraRelative}.XYZ)
   B=ActorPositionWS() {originType=CameraRelative}
  )
  n14 = CollectionParameter() {collection=&FortniteMaterialParameters, parameterName=PlayerPosition}
  n15 = StaticSwitchParameter(True=Subtract(A=@n13, B=ComponentMask(None=@n14) {b=true, g=true, r=true}), False=@n13) {defaultValue=true, parameterName=Enable Player Attach}
  n16 = Multiply(
   A=VectorParameter() {parameterName=Size, defaultValue=<rgba 95 95 95 1>}.RGB
   B=MaterialFunctionCall() {materialFunction=&ObjectScale}.Scale XYZ
  )
  n17 = Saturate(None=Distance(A=Divide(A=@n15, B=@n16) {constB=100}, B=Constant()))
  n18 = OneMinus(None=Multiply(A=@n17, B=@n17))
  n19 = Multiply(A=@n18, B=@n18)
  ShadingPathSwitch(Default=LinearInterpolate(A=Saturate(None=@n10), Alpha=OneMinus(None=@n19)) {constA=0.5}, Mobile=Constant() {r=1})
 ]
}
