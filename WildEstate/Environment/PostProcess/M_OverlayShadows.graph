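# M_OverlayShadows is a translucent, unlit, two-sided material that, judging by its name, draws fake overlay shadows.
# An offset sum pushes vertices facing a fixed local direction along it, adding normal and camera-space shifts scaled by parameters.
# An Opacity parameter multiplies scene color, with a feature-level switch using Opacity alone on ES3.1.
# View-facing and direction masks and a black color parameter also appear, but material output connections are not shown.
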Material /WildEstate/Environment/PostProcess/M_OverlayShadows
graph {
 properties {bAutomaticallySetUsageInEditor=false, blendMode=BLEND_Translucent, shadingModel=MSM_Unlit, twoSided=true, usedShadingModels=Unlit}
 nodes [
  n1 = Transform(None=Constant3Vector() {constant=<rgba 1 0 -0.5 0>}) {transformSourceType=TRANSFORMSOURCE_Local}
  n2 = ScalarParameter() {defaultValue=13.2868, parameterName=ShadowOffset}
  n3 = VertexNormalWS()
  n4 = Saturate(None=DotProduct(A=@n1, B=@n3))
  n5 = ScalarParameter() {defaultValue=-1, parameterName=ShadowOffseVTXOffset}
  n6 = Transform(None=VectorParameter() {parameterName=CamDir, defaultValue=<rgba 0 0 1 1>}.RGB) {transformSourceType=TRANSFORMSOURCE_Camera}
  n7 = ScalarParameter() {defaultValue=0.7061, parameterName=ShadowOffsetCamOffset}
  Add(A=Multiply(A=Multiply(A=@n1, B=@n2), B=Ceil(None=@n4)), B=Add(A=Multiply(A=@n3, B=@n5), B=Multiply(A=@n6, B=@n7)))
  Constant() {r=1}
  n8 = ScalarParameter() {defaultValue=0.5, parameterName=Opacity}
  FeatureLevelSwitch(Default=Multiply(A=@n8, B=SceneColor!(UVs).RGB), ES3_1=@n8)
  n9 = WorldPosition()
  n10 = CameraPositionWS()
  n11 = DotProduct(A=VertexNormalWS(), B=Normalize(VectorInput=Subtract(A=@n9.XYZ, B=@n10)))
  Multiply(A=@n11) {constB=-1}
  Round(None=Saturate(None=@n11))
  Round(None=@n4)
  VectorParameter() {defaultValue=<rgba 0 0 0 1>}
 ]
}
