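# Translucent unlit overlay material that computes depth-based lines from scene depth, though no material output connections appear.
# Four diagonal depth taps from BlurSampleOffsets and the center depth feed combine4Samples, with bias, scale and divisor inputs.
# Parameters blend across three distance bands using two depth fades.
# FeatureLevelSwitch selects separate ES3_1 mobile parameters, including an unbanded line thickness, while the default path scales thickness by view size.
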Material /WildEstate/Environment/PostProcess/M_WildEstate_OverlayPost
graph {
 properties {bAutomaticallySetUsageInEditor=false, blendMode=BLEND_Translucent, shadingModel=MSM_Unlit, usedShadingModels=Unlit}
 nodes [
  Constant() {r=1}
  n1 = SceneTexture() {sceneTextureId=PPI_SceneDepth}
  n2 = ScalarParameter() {defaultValue=40000, group=SteppedTransition_Mobile, parameterName=Step1_Mobile_End}
  n3 = Saturate(None=Divide(
    A=Subtract(A=@n1.Color, B=@n2)
    B=ScalarParameter() {defaultValue=40000, group=SteppedTransition_Mobile, parameterName=Step1_Mobile_FadeDistance}
   ))
  n4 = LinearInterpolate(
   A=ScalarParameter() {defaultValue=1, group=Mobile, parameterName=BaseSampleBias_Mobile}
   B=ScalarParameter() {defaultValue=1, group=Mobile_Step2, parameterName=BaseSampleBias_Mobile_Step2}
   Alpha=@n3
  )
  n5 = ScalarParameter() {defaultValue=1, group=Mobile_Step3, parameterName=BaseSampleBias_Mobile_Step3}
  n6 = SceneTexture() {sceneTextureId=PPI_SceneDepth}
  n7 = ScalarParameter() {defaultValue=40000, group=SteppedTransition_Mobile, parameterName=Step2_Mobile_End}
  n8 = Saturate(None=Divide(
    A=Subtract(A=@n6.Color, B=@n7)
    B=ScalarParameter() {defaultValue=40000, group=SteppedTransition_Mobile, parameterName=Step2_Mobile_FadeDistance}
   ))
  n9 = LinearInterpolate(
   A=ScalarParameter() {defaultValue=1, group=Mobile, parameterName=ScaleSample_Mobile}
   B=ScalarParameter() {defaultValue=1, group=Mobile_Step2, parameterName=ScaleSample_Mobile_Step2}
   Alpha=@n3
  )
  n10 = ScalarParameter() {defaultValue=1, group=Mobile_Step3, parameterName=ScaleSample_Mobile_Step3}
  n11 = ComponentMask(None=ScreenPosition().ViewportUV) {g=true, r=true}
  n12 = Multiply(A=NamedRerouteUsage(), B=Divide(A=ViewSize(), B=NamedRerouteUsage()))
  n13 = SceneTexelSize()
  n14 = ScalarParameter() {defaultValue=2, group=Mobile, parameterName=LineThicknessMobile}
  n15 = SceneTexelSize()
  n16 = MaterialFunctionCall!(OffsetAmount=FeatureLevelSwitch(Default=Multiply(A=@n12, B=@n13), ES3_1=Multiply(A=@n14, B=@n15))) {materialFunction=/Engine/Functions/Engine_MaterialFunctions01/ImageAdjustment/BlurSampleOffsets}
  n17 = LinearInterpolate(
   A=ScalarParameter() {defaultValue=1, group=Mobile, parameterName=DivisorDepthSample_Mobile}
   B=ScalarParameter() {defaultValue=1, group=Mobile_Step2, parameterName=DivisorDepthSample_Mobile_Step2}
   Alpha=@n3
  )
  n18 = ScalarParameter() {defaultValue=1, group=Mobile_Step3, parameterName=DivisorDepthSample_Mobile_Step3}
  MaterialFunctionCall!(
   Base=SceneTexture() {sceneTextureId=PPI_SceneDepth}.Color
   BaseSampleBias=FeatureLevelSwitch(Default=NamedRerouteUsage(), ES3_1=LinearInterpolate(A=@n4, B=@n5, Alpha=@n8))
   Scale=FeatureLevelSwitch(Default=NamedRerouteUsage(), ES3_1=LinearInterpolate(A=@n9, B=@n10, Alpha=@n8))
   Sample1=SceneTexture(UVs=Add(A=@n11, B=@n16."-.5,.5") {desc=x+}) {sceneTextureId=PPI_SceneDepth}.Color
   Sample2=SceneTexture(UVs=Add(A=@n11, B=@n16.".5,.5") {desc=y+}) {sceneTextureId=PPI_SceneDepth}.Color
   Sample3=SceneTexture(UVs=Add(A=@n11, B=@n16.".5,-.5") {desc=x-}) {sceneTextureId=PPI_SceneDepth}.Color
   Sample4=SceneTexture(UVs=Add(A=@n11, B=@n16."-.5,-.5") {desc=y-}) {sceneTextureId=PPI_SceneDepth}.Color
   Divisor=FeatureLevelSwitch(Default=NamedRerouteUsage(), ES3_1=LinearInterpolate(A=@n17, B=@n18, Alpha=@n8))
   UseCenteredLines=Constant() {r=1}
   UseRGB
  ) {materialFunction=/Game/Packages/jl_tests/combine4Samples}
  Multiply(A=VertexNormalWS(), B=ScalarParameter() {defaultValue=27.2804, parameterName=Push})
  n19 = ScalarParameter() {defaultValue=1, group=DepthSampling, parameterName=ScaleSample_1}
  n20 = ScalarParameter() {defaultValue=1, group=DepthSampling_2, parameterName=ScaleSample_2}
  n21 = NamedRerouteUsage()
  NamedRerouteDeclaration(None=LinearInterpolate(
    A=LinearInterpolate(A=@n19, B=@n20, Alpha=@n21)
    B=ScalarParameter() {defaultValue=1, group=DepthSampling_3, parameterName=ScaleSample_3}
    Alpha=NamedRerouteUsage()
   )) {name=ScaleSamples, nodeColor=<rgba 0 1 0.2969 1>}
  n22 = ScalarParameter() {defaultValue=1, group=DepthSampling, parameterName=BaseSampleBias_1}
  n23 = ScalarParameter() {defaultValue=1, group=DepthSampling_2, parameterName=BaseSampleBias_2}
  n24 = NamedRerouteUsage()
  NamedRerouteDeclaration(None=LinearInterpolate(
    A=LinearInterpolate(A=@n22, B=@n23, Alpha=@n24)
    B=ScalarParameter() {defaultValue=1, group=DepthSampling_3, parameterName=BaseSampleBias_3}
    Alpha=NamedRerouteUsage()
   )) {name=BaseSampleBias, nodeColor=<rgba 1 0 0.0234 1>}
  n25 = ScalarParameter() {defaultValue=-800, group=DepthSampling, parameterName=DivisorDepthSamples}
  n26 = ScalarParameter() {defaultValue=-800, group=DepthSampling_2, parameterName=DivisorDepthSamples2}
  n27 = NamedRerouteUsage()
  NamedRerouteDeclaration(None=LinearInterpolate(
    A=LinearInterpolate(A=@n25, B=@n26, Alpha=@n27)
    B=ScalarParameter() {defaultValue=-800, group=DepthSampling_3, parameterName=DivisorDepthSamples3}
    Alpha=NamedRerouteUsage()
   )) {name=DivisoDepthSamples, nodeColor=<rgba 0 1 0.6953 1>}
  n28 = ScalarParameter() {defaultValue=1, group=DepthSampling, parameterName=LineThickness_1}
  n29 = ScalarParameter() {defaultValue=1, group=DepthSampling_2, parameterName=LineThickness_2}
  n30 = NamedRerouteUsage()
  NamedRerouteDeclaration(None=LinearInterpolate(
    A=LinearInterpolate(A=@n28, B=@n29, Alpha=@n30)
    B=ScalarParameter() {defaultValue=1, group=DepthSampling_3, parameterName=LineThickness_3}
    Alpha=NamedRerouteUsage()
   )) {name=LineThickness, nodeColor=<rgba 1 0.7734 0 1>}
  n31 = Subtract(
   A=SceneTexture() {sceneTextureId=PPI_SceneDepth}.Color
   B=ScalarParameter() {defaultValue=2000, group=SteppedTransition, parameterName=Step1_End}
  )
  n32 = ScalarParameter() {defaultValue=1, group=SteppedTransition, parameterName=Step1_FadeDistance}
  NamedRerouteDeclaration(None=Saturate(None=Divide(A=@n31, B=@n32))) {name=Step1Blend, nodeColor=<rgba 0.4531 0 1 1>}
  n33 = ScalarParameter() {defaultValue=2000, group=DepthSampling, parameterName=BiasLineThickness_1}
  n34 = ScalarParameter() {defaultValue=2000, group=DepthSampling_2, parameterName=BiasLineThickness_2}
  n35 = NamedRerouteUsage()
  NamedRerouteDeclaration(None=LinearInterpolate(
    A=LinearInterpolate(A=@n33, B=@n34, Alpha=@n35)
    B=ScalarParameter() {defaultValue=2000, group=DepthSampling_3, parameterName=BiasLineThickness_3}
    Alpha=NamedRerouteUsage()
   )) {name=BaseLineThickness, nodeColor=<rgba 1 0.4453 0 1>}
  n36 = Subtract(
   A=SceneTexture() {sceneTextureId=PPI_SceneDepth}.Color
   B=ScalarParameter() {defaultValue=40000, group=SteppedTransition, parameterName=Step2_End}
  )
  n37 = ScalarParameter() {defaultValue=40000, group=SteppedTransition, parameterName=Step2_FadeDistance}
  NamedRerouteDeclaration(None=Saturate(None=Divide(A=@n36, B=@n37))) {name=Step2Blend, nodeColor=<rgba 0.4531 0 1 1>}
  SceneDepth()
  SceneTexture() {sceneTextureId=PPI_Roughness}
 ]
}
