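# Test translucent, unlit material that pairs a flat orange-red color with a depth-based fade, stored among the post-process assets.
# A DepthFade node takes full constant opacity and softens it near intersections with opaque scene geometry, with a scalar parameter setting the fade distance.
# The color is a fixed constant, and the graph samples no textures.
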Material /WildEstate/Environment/PostProcess/M_testTranslucent
graph {
 properties {
  bAllowFrontLayerTranslucency=false
  bAutomaticallySetUsageInEditor=false
  blendMode=BLEND_Translucent
  shadingModel=MSM_Unlit
  usedShadingModels=Unlit
 }
 nodes [
  Constant3Vector() {constant=<rgba 1 0.1549 0.0813 1>}
  DepthFade(Opacity=Constant() {r=1}, FadeDistance=ScalarParameter() {defaultValue=336.6769})
 ]
}
