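# Test post-process material that experiments with compensating for temporal AA jitter when reading custom depth.
# It samples custom depth at the pixel position minus the temporal sample offset, divided by view size.
# That depth, scaled down and saturated, multiplies scene color tinted by parameter C.
# Loose experimental branches cover jittered UVs, scene depth, texel-size offsets and DitherTemporalAA, and no material output connection is shown.
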Material /WildEstate/Environment/PostProcess/PPM_Test
graph {
 properties {bAutomaticallySetUsageInEditor=false, materialDomain=MD_PostProcess, usedShadingModels=Default Lit}
 nodes [
  n1 = ViewProperty() {property=MEVP_TemporalSampleOffset}
  n2 = ScalarParameter() {defaultValue=0.0001, parameterName=test}
  Add(A=ScreenPosition().ViewportUV, B=Multiply(A=@n1.Property, B=Divide(A=@n2) {constB=10000}))
  n3 = ViewProperty() {property=MEVP_ViewSize}
  Divide(A=@n1.Property, B=@n3.Property)
  Divide(A=ComponentMask(None=SceneTexture() {bFiltered=true, sceneTextureId=PPI_SceneDepth}.Color) {r=true}) {constB=1000}
  MaterialFunctionCall() {materialFunction=&DitherTemporalAA}
  n4 = SceneTexture() {sceneTextureId=PPI_PostProcessInput0}
  n5 = Subtract(A=ScreenPosition().PixelPosition, B=ViewProperty() {property=MEVP_TemporalSampleOffset}.Property)
  n6 = ComponentMask(None=SceneTexture(UVs=Divide(A=@n5, B=@n3.Property)) {bFiltered=true, sceneTextureId=PPI_CustomDepth}.Color) {r=true}
  Multiply(
   A=Multiply(A=VectorParameter() {parameterName=C, defaultValue=<rgba 0.2539 1 0.5465 1>}.RGB, B=ComponentMask(None=@n4.Color) {b=true, g=true, r=true})
   B=Saturate(None=Divide(A=@n6) {constB=3423423})
  )
  n7 = SceneTexelSize()
  n8 = ViewProperty() {property=MEVP_TemporalSampleIndex}
  Multiply(A=Multiply(A=@n7, B=@n8.InvProperty)) {constB=-1}
  Multiply(A=@n7, B=@n8.Property)
  ViewProperty() {property=MEVP_TemporalSampleOffset}
 ]
}
