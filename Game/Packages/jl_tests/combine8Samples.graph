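# Material function that turns a base value and eight samples into a scalar outline mask, labeled as depth outline.
# It sums the differences between the biased base and each sample, then divides by the divisor and its negation to handle both signs in one float2.
# Both channels are clamped, inverted, scaled, then clamped and inverted again.
# UseCenteredLines weights the second channel before a Max, enabling outlines on both sides of discontinuities.
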MaterialFunction /Game/Packages/jl_tests/combine8Samples
graph {nodes [
  Constant() {desc=1, r=2}
  n1 = Add(
   A=FunctionInput() {bUsePreviewValueAsDefault=true, inputName=BaseSampleBias, inputType=FunctionInput_Scalar, sortPriority=10}
   B=FunctionInput() {bUsePreviewValueAsDefault=true, inputName=Base, inputType=FunctionInput_Scalar}
  )
  n2 = Subtract(A=@n1, B=FunctionInput() {bUsePreviewValueAsDefault=true, inputName=Sample1, inputType=FunctionInput_Scalar, sortPriority=1})
  n3 = Subtract(A=@n1, B=FunctionInput() {bUsePreviewValueAsDefault=true, inputName=Sample2, inputType=FunctionInput_Scalar, sortPriority=2})
  n4 = Subtract(A=@n1, B=FunctionInput() {bUsePreviewValueAsDefault=true, inputName=Sample3, inputType=FunctionInput_Scalar, sortPriority=3})
  n5 = Subtract(A=@n1, B=FunctionInput() {bUsePreviewValueAsDefault=true, inputName=Sample4, inputType=FunctionInput_Scalar, sortPriority=4})
  n6 = Subtract(A=@n1, B=FunctionInput() {bUsePreviewValueAsDefault=true, inputName=Sample5, inputType=FunctionInput_Scalar, sortPriority=6})
  n7 = Subtract(A=@n1, B=FunctionInput() {bUsePreviewValueAsDefault=true, inputName=Sample6, inputType=FunctionInput_Scalar, sortPriority=7})
  n8 = Subtract(A=@n1, B=FunctionInput() {bUsePreviewValueAsDefault=true, inputName=Sample7, inputType=FunctionInput_Scalar, sortPriority=8})
  n9 = Subtract(A=@n1, B=FunctionInput() {bUsePreviewValueAsDefault=true, inputName=Sample8, inputType=FunctionInput_Scalar, sortPriority=9})
  n10 = Add(A=Add(A=Add(A=@n2, B=@n3), B=Add(A=@n4, B=@n5)), B=Add(A=Add(A=@n6, B=@n7), B=Add(A=@n8, B=@n9)))
  n11 = FunctionInput() {bUsePreviewValueAsDefault=true, inputName=Divisor, inputType=FunctionInput_Scalar, sortPriority=11, previewValue=<-64 0 0 1>}
  n12 = Divide(
   A=MaterialFunctionCall(X=@n10, Y=@n10) {materialFunction=&MakeFloat2}.Result
   B=MaterialFunctionCall(X=@n11, Y=Multiply(A=@n11) {constB=-1}) {materialFunction=&MakeFloat2}.Result
  )
  n13 = FunctionInput() {bUsePreviewValueAsDefault=true, inputName=Scale, inputType=FunctionInput_Scalar, sortPriority=12, previewValue=<2 0 0 1>}
  n14 = Multiply(A=OneMinus(None=Clamp(None=@n12) {desc=depth outline}), B=@n13)
  n15 = MaterialFunctionCall(Float2=OneMinus(None=Clamp(None=@n14) {desc=depth outline})) {materialFunction=&BreakOutFloat2Components}
  n16 = FunctionInput(Preview=ScalarParameter() {desc=Used to draw outlines on both sides of a discontinuity, group=Cartoon, parameterName=UseCenteredLines}) {bUsePreviewValueAsDefault=true, inputName=UseCenteredLines, inputType=FunctionInput_Scalar, sortPriority=15}
  FunctionOutput(None=Max(A=@n15.R, B=Multiply(A=@n15.G, B=@n16)))
 ]}
