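# WildEstate post-process material, blended after DOF, that draws stylized scene outlines and adds distance fog and screen-space shadows.
# Edge detection compares neighboring depth, normal and roughness samples, with distance-stepped tuning, custom-depth masking and far fade.
# Custom HLSL ray-marches depth for shadows cast by shading-model-ID-4 pixels; fog color follows time of day.
# ES3.1 switches simplify outlines and skip shadows; debug and experimental blur and quantize branches remain.
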Material /WildEstate/Environment/PostProcess/PPM_WildEstate
graph {
 properties {bAutomaticallySetUsageInEditor=false, blendableLocation=BL_SceneColorAfterDOF, materialDomain=MD_PostProcess, usedShadingModels=Default Lit}
 nodes [
  n1 = NamedRerouteUsage()
  n2 = NamedRerouteUsage()
  MaterialExpressionSceneTexture_54 = SceneTexture() {sceneTextureId=PPI_SceneDepth}
  MaterialExpressionScreenPosition_8 = ScreenPosition()
  MaterialExpressionNormalize_1 = Normalize(VectorInput=VectorParameter() {parameterName=Ray_Dir, defaultValue=<rgba 1 1 -0.25 1>}.RGB)
  MaterialExpressionScalarParameter_64 = ScalarParameter() {defaultValue=32, parameterName=Ray_Steps}
  MaterialExpressionScalarParameter_63 = ScalarParameter() {defaultValue=0.01, parameterName=Ray_Bias}
  n3 = ComponentMask(None=@MaterialExpressionSceneTexture_54.Color) {r=true}
  n4 = ScalarParameter() {defaultValue=200, parameterName=DepthFadeRaysOffset}
  n5 = Saturate(None=Divide(A=Subtract(A=@n3, B=@n4), B=ScalarParameter() {defaultValue=2000, parameterName=DepthFadeRays}))
  MaterialExpressionMultiply_28 = Multiply(A=OneMinus(None=@n5), B=ScalarParameter() {defaultValue=0.05, parameterName=Ray_StepSize})
  MaterialExpressionScalarParameter_66 = ScalarParameter() {defaultValue=16, parameterName=Ray_MaxDepth}
  n6 = Custom!(
   Tex=@MaterialExpressionSceneTexture_54.Color
   UV=@MaterialExpressionScreenPosition_8.ViewportUV
   TraceVec=@MaterialExpressionNormalize_1
   NumSteps=@MaterialExpressionScalarParameter_64
   bias=@MaterialExpressionScalarParameter_63
   StepSizeMAX=@MaterialExpressionMultiply_28
   MaxDepth=@MaterialExpressionScalarParameter_66
  ) {
   code="// The below expression will get compiled\r\n// into the output of this node\r\n//float3(1, 1, 1);\r\n//float3 Color = SceneTextureLookup(ViewportUVToSceneTextureUV(UV, 1), 1, false);\r\n\r\nfloat TimeLerp = 1;\r\nfloat DepthDiff = 0;\r\nfloat DepthDiff2 = 0;\r\n\r\nfloat LastDiff = -bias;\r\nfloat StepSize = StepSizeMAX/NumSteps;\r\nDebug = 1;\r\nfloat3 RayStepUVz = float3(TraceVec.x, TraceVec.y, TraceVec.z);\r\n\r\nfloat accum = 0;\r\nfloat3 RayUVz = float3(UV, 0 );\r\nfloat3 RayUVzInv = float3(UV, 0 );\r\nfloat depth = SceneTextureLookup(ViewportUVToSceneTextureUV(UV, 1), 1, false);\r\nint i = 0;\r\nwhile (i < NumSteps)\r\n{\r\n\r\n\tRayUVz += RayStepUVz*StepSize;\r\n    RayUVzInv -= RayStepUVz*StepSize;\r\n    RayUVz.xy = saturate(RayUVz.xy);\r\n    RayUVzInv.xy = saturate(RayUVzInv.xy);\r\n\r\n    float MaskSample1 = SceneTextureLookup(ViewportUVToSceneTextureUV(RayUVz.xy, 23), 23, false);\r\n    if (MaskSample1 == 4)\r\n\t\t{\r\n\r\n        MaskSample1 = 1;\r\n        }\r\n    else\r\n\t\t{\r\n\r\n        MaskSample1 = 0;\r\n        }  \r\n    float MaskSample2 = SceneTextureLookup(ViewportUVToSceneTextureUV(RayUVzInv.xy, 23), 23, false);\r\n    if (MaskSample2 == 4)\r\n\t\t{\r\n\r\n        MaskSample2 = 1;\r\n        }\r\n    else\r\n\t\t{\r\n\r\n        MaskSample2 = 0;\r\n        }  \r\n\tfloat SampleDepth = SceneTextureLookup(ViewportUVToSceneTextureUV(RayUVz.xy, 1), 1, false);\r\n    float SampleDepth2 = SceneTextureLookup(ViewportUVToSceneTextureUV(RayUVzInv.xy, 1), 1, false);\r\n\t//DepthDiff = (SampleDepth+RayUVz.z)-depth + max((SampleDepth2-depth),0);\r\n   // DepthDiff = (SampleDepth+RayUVz.z)-depth + (SampleDepth2-depth);\r\n    DepthDiff = (SampleDepth+RayUVz.z)-depth ;\r\n    DepthDiff *= MaskSample1; \r\n    //DepthDiff = (SampleDepth+RayUVz.z)-depth;\r\n    //DepthDiff = (SampleDepth+RayUVz.z)-depth;\r\n    DepthDiff2 = (SampleDepth2+RayUVzInv.z)-depth;\r\n    DepthDiff2 *= MaskSample2; \r\n    \r\n\r\n\t\tif (DepthDiff < 0 &&&& DepthDiff > -MaxDepth)\r\n\t\t{\r\n\r\n\t\t\t//if(LastDiff > 0)\r\n\t\t\t//{\r\n\t\t\t//\tTimeLerp = saturate( LastDiff / (LastDiff - DepthDiff));\r\n\t\t\t//\taccum += StepSize * (1-TimeLerp);\r\n\t\t\t\t//accum+=StepSize;\r\n\t\t\t//}\r\n\t\t\t//else\r\n\t\t\t//{\r\n\t\t\t//\taccum+=StepSize;\r\n\t\t\t//}\r\n            accum += 0.1;\r\n\r\n\t\t}\r\n\t\telse\r\n\t\t//if(LastDiff <= 0)\r\n\t\t{\r\n\t\t\t//accum -=0.1;\r\n\t\t}\r\n //       if (DepthDiff2 < 0 )\r\n\t//\t{\r\n   //        accum +=0.1;\r\n        \r\n    //    }\r\n\r\n\t//LastDiff = DepthDiff;\r\n\r\n\ti++;\r\n}\r\n//Debug =  SceneTextureLookup(GetDefaultSceneTextureUV(Parameters, 23), 23, false);\r\nreturn 1-saturate(accum);\r\n"
   showCode=true
   additionalOutputs [{outputName=Debug, outputType=CMOT_Float1}]
   inputs [
    {
     inputName=Tex
     input {expression=&self:MaterialExpressionSceneTexture_54, mask=1, maskA=1, maskB=1, maskG=1, maskR=1}
    }
    {inputName=UV, input {expression=&self:MaterialExpressionScreenPosition_8}}
    {inputName=TraceVec, input {expression=&self:MaterialExpressionNormalize_1}}
    {inputName=NumSteps, input {expression=&self:MaterialExpressionScalarParameter_64}}
    {inputName=bias, input {expression=&self:MaterialExpressionScalarParameter_63}}
    {inputName=StepSizeMAX, input {expression=&self:MaterialExpressionMultiply_28}}
    {inputName=MaxDepth, input {expression=&self:MaterialExpressionScalarParameter_66}}
   ]
  }
  n7 = Round(None=@n6.return)
  n8 = ScalarParameter() {defaultValue=0.5, parameterName=Ray_ShadowMin}
  n9 = Multiply(A=@n1, B=Clamp(None=Add(A=@n2, B=@n7), Min=Add(A=@n8, B=@n5)))
  Add(
   A=NamedRerouteUsage()
   B=StaticSwitchParameter(True=FeatureLevelSwitch(Default=@n9, ES3_1=@n1), False=@n1) {defaultValue=true, parameterName=UseScreenSpaceShadows}
  )
  Add(A=Add(A=Add(), B=Add())) {constB=0.05}
  Constant2Vector()
  Constant()
  Custom() {
   code="// The below expression will get compiled\r\n// into the output of this node\r\n//float3(1, 1, 1);\r\n//float3 Color = SceneTextureLookup(ViewportUVToSceneTextureUV(UV, 1), 1, false);\r\n\r\nfloat TimeLerp = 1;\r\nfloat DepthDiff = 0;\r\nfloat DepthDiff2 = 0;\r\n\r\nfloat LastDiff = -bias;\r\nfloat StepSize = StepSizeMAX/NumSteps;\r\nDebug = 1;\r\nfloat3 RayStepUVz = float3(TraceVec.x, TraceVec.y, TraceVec.z);\r\n\r\nfloat accum = 0;\r\nfloat3 RayUVz = float3(UV, 0 );\r\nfloat3 RayUVzInv = float3(UV, 0 );\r\nfloat depth = SceneTextureLookup(ViewportUVToSceneTextureUV(UV, 1), 1, false);\r\nint i = 0;\r\nwhile (i < NumSteps)\r\n{\r\n\r\n\tRayUVz += RayStepUVz*StepSize;\r\n    RayUVzInv -= RayStepUVz*StepSize;\r\n    RayUVz.xy = saturate(RayUVz.xy);\r\n    RayUVzInv.xy = saturate(RayUVzInv.xy);\r\n\r\n    float MaskSample1 = SceneTextureLookup(ViewportUVToSceneTextureUV(RayUVz.xy, 23), 23, false);\r\n    if (MaskSample1 == 4)\r\n\t\t{\r\n\r\n        MaskSample1 = 1;\r\n        }\r\n    else\r\n\t\t{\r\n\r\n        MaskSample1 = 0;\r\n        }  \r\n    float MaskSample2 = SceneTextureLookup(ViewportUVToSceneTextureUV(RayUVzInv.xy, 23), 23, false);\r\n    if (MaskSample2 == 4)\r\n\t\t{\r\n\r\n        MaskSample2 = 1;\r\n        }\r\n    else\r\n\t\t{\r\n\r\n        MaskSample2 = 0;\r\n        }  \r\n\tfloat SampleDepth = SceneTextureLookup(ViewportUVToSceneTextureUV(RayUVz.xy, 1), 1, false);\r\n    float SampleDepth2 = SceneTextureLookup(ViewportUVToSceneTextureUV(RayUVzInv.xy, 1), 1, false);\r\n\t//DepthDiff = (SampleDepth+RayUVz.z)-depth + max((SampleDepth2-depth),0);\r\n    DepthDiff = (SampleDepth+RayUVz.z)-depth + (SampleDepth2-depth);\r\n    DepthDiff *= MaskSample1; \r\n    //DepthDiff = (SampleDepth+RayUVz.z)-depth;\r\n    //DepthDiff = (SampleDepth+RayUVz.z)-depth;\r\n    DepthDiff2 = (SampleDepth2+RayUVzInv.z)-depth;\r\n    DepthDiff2 *= MaskSample2; \r\n    \r\n\r\n\t\tif (DepthDiff < 0 &&&& DepthDiff > -MaxDepth)\r\n\t\t{\r\n\r\n\t\t\t//if(LastDiff > 0)\r\n\t\t\t//{\r\n\t\t\t//\tTimeLerp = saturate( LastDiff / (LastDiff - DepthDiff));\r\n\t\t\t//\taccum += StepSize * (1-TimeLerp);\r\n\t\t\t\t//accum+=StepSize;\r\n\t\t\t//}\r\n\t\t\t//else\r\n\t\t\t//{\r\n\t\t\t//\taccum+=StepSize;\r\n\t\t\t//}\r\n            accum += 0.1;\r\n\r\n\t\t}\r\n\t\telse\r\n\t\t//if(LastDiff <= 0)\r\n\t\t{\r\n\t\t\taccum -=0.1;\r\n\t\t}\r\n //       if (DepthDiff2 < 0 )\r\n\t//\t{\r\n   //        accum +=0.1;\r\n        \r\n    //    }\r\n\r\n\t//LastDiff = DepthDiff;\r\n\r\n\ti++;\r\n}\r\n//Debug =  SceneTextureLookup(GetDefaultSceneTextureUV(Parameters, 23), 23, false);\r\nreturn 1-saturate(accum);\r\n"
   showCode=true
   additionalOutputs [{outputName=Debug, outputType=CMOT_Float1}]
   inputs [
    {inputName=Tex, input {expression=null, mask=1, maskA=1, maskB=1, maskG=1, maskR=1}}
    {inputName=UV, input {expression=null}}
    {inputName=TraceVec, input {expression=null}}
    {inputName=NumSteps, input {expression=null}}
    {inputName=bias, input {expression=null}}
    {inputName=StepSizeMAX, input {expression=null}}
    {inputName=MaxDepth, input {expression=null}}
   ]
  }
  n10 = Multiply(A=ComponentMask(None=NamedRerouteUsage()) {b=true, g=true, r=true}, B=ScalarParameter() {defaultValue=1, parameterName=Brigthness Offset})
  n11 = ScalarParameter() {defaultValue=5, parameterName=QuantizeHUE}
  n12 = Round(None=Multiply(A=Normalize(VectorInput=@n10), B=@n11))
  n13 = Length(None=@n10) {is=class MaterialExpressionLength}
  n14 = ScalarParameter() {defaultValue=5, parameterName=QuantizeValue}
  n15 = Divide(A=Round(None=Multiply(A=@n13, B=@n14)), B=@n14)
  n16 = ScalarParameter() {defaultValue=0.1, parameterName=QuantizeValue_min}
  Desaturation(None=Multiply(A=Divide(A=@n12, B=@n11), B=Max(A=@n15, B=@n16)), Fraction=ScalarParameter() {defaultValue=-0.1, parameterName=Saturation})
  n17 = SceneTexture() {sceneTextureId=PPI_PostProcessInput0}
  Desaturation(None=@n17.Color)
  n18 = SceneTexelSize()
  n19 = Min(A=ComponentMask(None=ViewProperty() {property=MEVP_ResolutionFraction}.Property) {r=true})
  Divide(A=@n18, B=@n19)
  n20 = NamedRerouteUsage()
  n21 = AppendVector(
   A=ScalarParameter() {defaultValue=0.01, group=Mobile, parameterName=TestMaxTexelSizeX}
   B=ScalarParameter() {defaultValue=0.01, group=Mobile, parameterName=TestMaxTexelSizeY}
  )
  n22 = NamedRerouteDeclaration(None=Multiply(A=@n20, B=Min(A=@n18, B=@n21))) {name=LineThckenesMain, nodeColor=<rgba 1 0 0.7969 1>}
  FeatureLevelSwitch(Default=@n22, ES3_1=Multiply(A=ScalarParameter() {defaultValue=2, group=Mobile, parameterName=LineThicknessMobile}, B=SceneTexelSize()))
  n23 = ComponentMask(None=ScreenPosition().ViewportUV) {g=true, r=true}
  n24 = Multiply(
   A=ScalarParameter() {defaultValue=2, group=CharacterMask, parameterName=CharacterMaskEdge}
   B=Divide(A=ViewSize(), B=ScalarParameter() {defaultValue=2000, group=CharacterMask, parameterName=CharacterMaskBaseLine})
  )
  n25 = SceneTexelSize()
  n26 = MaterialFunctionCall!(OffsetAmount=Saturate(None=Multiply(A=@n24, B=@n25))) {materialFunction=/Engine/Functions/Engine_MaterialFunctions01/ImageAdjustment/BlurSampleOffsets}
  n27 = SceneTexture(UVs=Add(A=@n23, B=@n26."-.5,.5") {desc=x+}) {sceneTextureId=PPI_CustomDepth}
  n28 = SceneTexture(UVs=Add(A=@n23, B=@n26.".5,.5") {desc=y+}) {sceneTextureId=PPI_CustomDepth}
  n29 = SceneTexture(UVs=Add(A=@n23, B=@n26.".5,-.5") {desc=x-}) {sceneTextureId=PPI_CustomDepth}
  n30 = SceneTexture(UVs=Add(A=@n23, B=@n26."-.5,-.5") {desc=y-}) {sceneTextureId=PPI_CustomDepth}
  n31 = Min(
   A=Min(A=ComponentMask(None=@n27.Color) {r=true}, B=ComponentMask(None=@n28.Color) {r=true})
   B=Min(A=ComponentMask(None=@n29.Color) {r=true}, B=ComponentMask(None=@n30.Color) {r=true})
  )
  Floor(None=@n31)
  MaterialFunctionCall() {materialFunction=&ScreenResolution}
  n32 = SceneTexelSize()
  n33 = ComponentMask(None=@n32) {r=true}
  n34 = ScalarParameter() {defaultValue=0.01, group=Mobile, parameterName=TestScreenPercentaget}
  n35 = Power(Base=Max(A=Divide(A=@n33, B=@n34)), Exp=ScalarParameter() {defaultValue=2, group=Mobile, parameterName=TestScreenPercentageExp})
  MaterialFunctionCall(Number=@n35, UVs=ScreenPosition().ViewportUV) {materialFunction=&DebugScalarValues}
  n36 = ComponentMask(None=ScreenPosition().ViewportUV) {g=true, r=true}
  n37 = ScalarParameter() {defaultValue=1.463, parameterName=OffsetFakeShadow}
  n38 = ComponentMask(None=SceneTexture(UVs=Add(A=@n36, B=@n37)) {sceneTextureId=PPI_SceneDepth}.Color) {r=true}
  n39 = ComponentMask(None=SceneTexture() {sceneTextureId=PPI_SceneDepth}.Color) {r=true}
  n40 = ComponentMask(None=SceneTexture(UVs=Subtract(A=@n36, B=@n37)) {sceneTextureId=PPI_SceneDepth}.Color) {r=true}
  MaterialFunctionCall!(
   Vector3=MaterialFunctionCall(X=Subtract(A=@n38, B=@n39), Y=Subtract(A=@n40, B=@n39), Z=NamedRerouteUsage()) {materialFunction=&MakeFloat3}.Result
   MaximumNumberOfDigits
   UVs
   DebugTextLocation RG_UpperRight BA_LowerLeft
   Component Spacing
  ) {materialFunction=/Engine/Functions/Engine_MaterialFunctions02/Utility/DebugFloat3Values}
  MaterialFunctionCall(Vector3=AppendVector(A=@n32, B=Constant())) {materialFunction=&DebugFloat3Values}
  n41 = SceneTexture(UVs=Add(A=@n23, B=@n26."0,1") {desc=x+}) {sceneTextureId=PPI_CustomDepth}
  n42 = SceneTexture(UVs=Add(A=@n23, B=@n26."1,0") {desc=y+}) {sceneTextureId=PPI_CustomDepth}
  n43 = SceneTexture(UVs=Add(A=@n23, B=@n26."0,-1") {desc=x-}) {sceneTextureId=PPI_CustomDepth}
  n44 = SceneTexture(UVs=Add(A=@n23, B=@n26."-1,0") {desc=y-}) {sceneTextureId=PPI_CustomDepth}
  n45 = Min(
   A=Min(A=ComponentMask(None=@n41.Color) {r=true}, B=ComponentMask(None=@n42.Color) {r=true})
   B=Min(A=ComponentMask(None=@n43.Color) {r=true}, B=ComponentMask(None=@n44.Color) {r=true})
  )
  Min(A=@n31, B=@n45)
  n46 = ComponentMask(None=ScreenPosition().ViewportUV) {g=true, r=true}
  n47 = MaterialFunctionCall(OffsetAmount=NamedRerouteUsage()) {materialFunction=&BlurSampleOffsets}
  n48 = SceneTexture(UVs=Add(A=@n46, B=@n47."-.5,.5") {desc=x+}) {sceneTextureId=PPI_WorldNormal}
  n49 = SceneTexture() {sceneTextureId=PPI_WorldNormal}
  n50 = SceneTexture(UVs=Add(A=@n46, B=@n47.".5,.5") {desc=y+}) {sceneTextureId=PPI_WorldNormal}
  n51 = SceneTexture(UVs=Add(A=@n46, B=@n47.".5,-.5") {desc=x-}) {sceneTextureId=PPI_WorldNormal}
  n52 = SceneTexture(UVs=Add(A=@n46, B=@n47."-.5,-.5") {desc=y-}) {sceneTextureId=PPI_WorldNormal}
  n53 = Min(
   A=Min(A=DotProduct(A=@n48.Color, B=@n49.Color), B=DotProduct(A=@n50.Color, B=@n49.Color))
   B=Min(A=DotProduct(A=@n49.Color, B=@n51.Color), B=DotProduct(A=@n49.Color, B=@n52.Color))
  )
  n54 = SceneTexture() {sceneTextureId=PPI_WorldNormal}
  n55 = SceneTexture(UVs=Add(A=@n47."0,1", B=@n46) {desc=x+}) {sceneTextureId=PPI_WorldNormal}
  n56 = SceneTexture(UVs=Add(A=@n47."1,0", B=@n46) {desc=y+}) {sceneTextureId=PPI_WorldNormal}
  n57 = SceneTexture(UVs=Add(A=@n47."0,-1", B=@n46) {desc=x-}) {sceneTextureId=PPI_WorldNormal}
  n58 = SceneTexture(UVs=Add(A=@n47."-1,0", B=@n46) {desc=y-}) {sceneTextureId=PPI_WorldNormal}
  n59 = Min(
   A=Min(A=DotProduct(A=@n54.Color, B=@n55.Color), B=DotProduct(A=@n54.Color, B=@n56.Color))
   B=Min(A=DotProduct(A=@n54.Color, B=@n57.Color), B=DotProduct(A=@n54.Color, B=@n58.Color))
  )
  Min(A=@n53, B=@n59)
  Min(A=SceneTexture() {sceneTextureId=PPI_CustomDepth}.Color, B=@n31)
  n60 = Constant()
  n61 = ScreenPosition()
  n62 = MaterialFunctionCall(OffsetAmount=Multiply(A=ScalarParameter() {defaultValue=5, parameterName=Roughness}, B=SceneTexelSize())) {materialFunction=&BlurSampleOffsets}
  n63 = SceneTexture(UVs=Add(A=@n61.ViewportUV, B=@n62."-.5,.5")) {sceneTextureId=PPI_Roughness}
  n64 = Constant()
  n65 = Add(
   A=If(A=@n63.Color, B=NamedRerouteUsage(), "A > B"=@n60, "A == B"=@n60, "A < B"=Constant() {r=1}) {constB=0.01}
   B=If(A=@n63.Color, B=NamedRerouteUsage(), "A > B"=Constant() {r=1}, "A == B"=@n64, "A < B"=@n64) {constB=0.01}
  )
  n66 = Constant()
  n67 = SceneTexture(UVs=Add(A=@n61.ViewportUV, B=@n62.".5,.5")) {sceneTextureId=PPI_Roughness}
  n68 = Constant()
  n69 = Add(
   A=If(A=@n67.Color, B=NamedRerouteUsage(), "A > B"=@n66, "A == B"=@n66, "A < B"=Constant() {r=1}) {constB=0.01}
   B=If(A=@n67.Color, B=NamedRerouteUsage(), "A > B"=Constant() {r=1}, "A == B"=@n68, "A < B"=@n68) {constB=0.01}
  )
  n70 = Constant()
  n71 = SceneTexture(UVs=Add(A=@n61.ViewportUV, B=@n62.".5,-.5")) {sceneTextureId=PPI_Roughness}
  n72 = Constant()
  n73 = Add(
   A=If(A=@n71.Color, B=NamedRerouteUsage(), "A > B"=@n70, "A == B"=@n70, "A < B"=Constant() {r=1}) {constB=0.01}
   B=If(A=@n71.Color, B=NamedRerouteUsage(), "A > B"=Constant() {r=1}, "A == B"=@n72, "A < B"=@n72) {constB=0.01}
  )
  n74 = Constant()
  n75 = SceneTexture(UVs=Add(A=@n61.ViewportUV, B=@n62."-.5,-.5")) {sceneTextureId=PPI_Roughness}
  n76 = Constant()
  n77 = Add(
   A=If(A=@n75.Color, B=NamedRerouteUsage(), "A > B"=@n74, "A == B"=@n74, "A < B"=Constant() {r=1}) {constB=0.01}
   B=If(A=@n75.Color, B=NamedRerouteUsage(), "A > B"=Constant() {r=1}, "A == B"=@n76, "A < B"=@n76) {constB=0.01}
  )
  Min(A=Min(A=@n65, B=@n69), B=Min(A=@n73, B=@n77))
  n78 = LinearInterpolate(
   A=ScalarParameter() {defaultValue=1, group=NormalSampling_1, parameterName=Normal_OutlineThickness_1}
   B=ScalarParameter() {defaultValue=1, group=NormalSampling_2, parameterName=Normal_OutlineThickness_2}
   Alpha=NamedRerouteUsage()
  )
  n79 = ScalarParameter() {defaultValue=1, group=NormalSampling_3, parameterName=Normal_OutlineThickness_3}
  n80 = NamedRerouteUsage()
  n81 = ViewSize()
  n82 = LinearInterpolate(
   A=LinearInterpolate(
    A=ScalarParameter() {defaultValue=2000, group=NormalSampling_1, parameterName=Normal_BaselineThickness_1}
    B=ScalarParameter() {defaultValue=2000, group=NormalSampling_2, parameterName=Normal_BaselineThickness_2}
    Alpha=NamedRerouteUsage()
   )
   B=ScalarParameter() {defaultValue=2000, group=NormalSampling_3, parameterName=Normal_BaselineThickness_3}
   Alpha=NamedRerouteUsage()
  )
  Multiply(A=Multiply(A=LinearInterpolate(A=@n78, B=@n79, Alpha=@n80), B=Divide(A=@n81, B=@n82)), B=SceneTexelSize())
  n83 = Divide(
   A=Subtract(
    A=SceneTexture() {sceneTextureId=PPI_SceneDepth}.Color
    B=ScalarParameter() {defaultValue=2000, group=RoughnessSampling, parameterName=Roughness_Step1_End}
   )
   B=ScalarParameter() {defaultValue=1, group=RoughnessSampling, parameterName=Roughness_Step1_FadeDistance}
  )
  Multiply(
   A=LinearInterpolate(
    A=ScalarParameter() {defaultValue=1, group=RoughnessSampling, parameterName=Roughness_OutlineThickness}
    B=ScalarParameter() {defaultValue=1, group=RoughnessSampling, parameterName=Roughness_OutlineThickness2}
    Alpha=Saturate(None=@n83)
   )
   B=SceneTexelSize()
  )
  Multiply(A=@n20, B=Divide(A=ViewSize(), B=NamedRerouteUsage()))
  Multiply(A=@n19, B=@n35)
  n84 = Multiply(A=NamedRerouteUsage(), B=NamedRerouteUsage())
  Multiply(A=@n84, B=NamedRerouteUsage())
  n85 = Max(A=Max(A=@n65, B=@n69), B=Max(A=@n73, B=@n77))
  Multiply(
   A=NamedRerouteDeclaration(None=Saturate(None=@n85)) {name=Roughness ZERO MASK, nodeColor=<rgba 0 1 0.6484 1>}
   B=ComponentMask(None=SceneTexture() {sceneTextureId=PPI_PostProcessInput0}.Color) {b=true, g=true, r=true}
  )
  NamedRerouteDeclaration(None=ScalarParameter() {parameterName=ThresholdBlur}) {name=Threshold, nodeColor=<rgba 0.7344 1 0 1>}
  n86 = ScalarParameter() {defaultValue=1, group=DepthSampling, parameterName=ScaleSample_1}
  n87 = ScalarParameter() {defaultValue=1, group=DepthSampling_2, parameterName=ScaleSample_2}
  n88 = NamedRerouteUsage()
  NamedRerouteDeclaration(None=LinearInterpolate(
    A=LinearInterpolate(A=@n86, B=@n87, Alpha=@n88)
    B=ScalarParameter() {defaultValue=1, group=DepthSampling_3, parameterName=ScaleSample_3}
    Alpha=NamedRerouteUsage()
   )) {name=ScaleSamples, nodeColor=<rgba 0 1 0.2969 1>}
  n89 = SceneTexture() {sceneTextureId=PPI_SceneDepth}
  n90 = ComponentMask(None=ScreenPosition().ViewportUV) {g=true, r=true}
  n91 = MaterialFunctionCall(OffsetAmount=@n22) {materialFunction=&BlurSampleOffsets}
  n92 = SceneTexture(UVs=Add(A=@n90, B=@n91."-.5,.5") {desc=x+}) {sceneTextureId=PPI_SceneDepth}
  n93 = SceneTexture(UVs=Add(A=@n90, B=@n91.".5,.5") {desc=y+}) {sceneTextureId=PPI_SceneDepth}
  n94 = SceneTexture(UVs=Add(A=@n90, B=@n91.".5,-.5") {desc=x-}) {sceneTextureId=PPI_SceneDepth}
  n95 = SceneTexture(UVs=Add(A=@n90, B=@n91."-.5,-.5") {desc=y-}) {sceneTextureId=PPI_SceneDepth}
  n96 = SceneTexture(UVs=Add(A=@n90, B=@n91."0,1") {desc=x+}) {sceneTextureId=PPI_SceneDepth}
  n97 = SceneTexture(UVs=Add(A=@n90, B=@n91."1,0") {desc=y+}) {sceneTextureId=PPI_SceneDepth}
  n98 = SceneTexture(UVs=Add(A=@n90, B=@n91."0,-1") {desc=x-}) {sceneTextureId=PPI_SceneDepth}
  n99 = SceneTexture(UVs=Add(A=@n90, B=@n91."-1,0") {desc=y-}) {sceneTextureId=PPI_SceneDepth}
  n100 = ScalarParameter() {defaultValue=1, group=Mobile, parameterName=BaseSampleBias_Mobile}
  n101 = ScalarParameter() {defaultValue=1, group=Mobile_Step2, parameterName=BaseSampleBias_Mobile_Step2}
  n102 = SceneTexture() {sceneTextureId=PPI_SceneDepth}
  n103 = ScalarParameter() {defaultValue=40000, group=SteppedTransition_Mobile, parameterName=Step1_Mobile_End}
  n104 = Saturate(None=Divide(
    A=Subtract(A=@n102.Color, B=@n103)
    B=ScalarParameter() {defaultValue=40000, group=SteppedTransition_Mobile, parameterName=Step1_Mobile_FadeDistance}
   ))
  n105 = SceneTexture() {sceneTextureId=PPI_SceneDepth}
  n106 = ScalarParameter() {defaultValue=40000, group=SteppedTransition_Mobile, parameterName=Step2_Mobile_End}
  n107 = Saturate(None=Divide(
    A=Subtract(A=@n105.Color, B=@n106)
    B=ScalarParameter() {defaultValue=40000, group=SteppedTransition_Mobile, parameterName=Step2_Mobile_FadeDistance}
   ))
  n108 = FeatureLevelSwitch(
   Default=NamedRerouteUsage()
   ES3_1=LinearInterpolate(
    A=LinearInterpolate(A=@n100, B=@n101, Alpha=@n104)
    B=ScalarParameter() {defaultValue=1, group=Mobile_Step3, parameterName=BaseSampleBias_Mobile_Step3}
    Alpha=@n107
   )
  )
  n109 = ScalarParameter() {defaultValue=50, parameterName=FresnelBias_DivisorDepth}
  n110 = Saturate(None=LinearInterpolate(
    A=ScalarParameter() {defaultValue=50, parameterName=Fresnel_EdgeMask_High}
    B=ScalarParameter() {defaultValue=50, parameterName=Fresnel_EdgeMask_low}
    Alpha=NamedRerouteUsage()
   ))
  n111 = ScalarParameter() {defaultValue=1, group=Mobile, parameterName=DivisorDepthSample_Mobile}
  n112 = ScalarParameter() {defaultValue=1, group=Mobile_Step2, parameterName=DivisorDepthSample_Mobile_Step2}
  n113 = FeatureLevelSwitch(
   Default=NamedRerouteUsage()
   ES3_1=LinearInterpolate(
    A=LinearInterpolate(A=@n111, B=@n112, Alpha=@n104)
    B=ScalarParameter() {defaultValue=1, group=Mobile_Step3, parameterName=DivisorDepthSample_Mobile_Step3}
    Alpha=@n107
   )
  )
  n114 = FeatureLevelSwitch(Default=Multiply(A=LinearInterpolate(B=@n109, Alpha=@n110) {constA=1}, B=@n113), ES3_1=@n113)
  n115 = ScalarParameter() {defaultValue=50, parameterName=FresnelBias_ScaleSamples}
  n116 = ScalarParameter() {defaultValue=1, group=Mobile, parameterName=ScaleSample_Mobile}
  n117 = ScalarParameter() {defaultValue=1, group=Mobile_Step2, parameterName=ScaleSample_Mobile_Step2}
  n118 = FeatureLevelSwitch(
   Default=NamedRerouteUsage()
   ES3_1=LinearInterpolate(
    A=LinearInterpolate(A=@n116, B=@n117, Alpha=@n104)
    B=ScalarParameter() {defaultValue=1, group=Mobile_Step3, parameterName=ScaleSample_Mobile_Step3}
    Alpha=@n107
   )
  )
  n119 = FeatureLevelSwitch(Default=Multiply(A=LinearInterpolate(B=@n115, Alpha=@n110) {constA=1}, B=@n118), ES3_1=@n118)
  n120 = Constant() {r=1}
  n121 = ComponentMask(None=StaticSwitchParameter(
    True=MaterialFunctionCall!(
     Base=@n89.Color
     Sample1=@n92.Color
     Sample2=@n93.Color
     Sample3=@n94.Color
     Sample4=@n95.Color
     Sample5=@n96.Color
     Sample6=@n97.Color
     Sample7=@n98.Color
     Sample8=@n99.Color
     BaseSampleBias=@n108
     Divisor=@n114
     Scale=@n119
     UseCenteredLines=@n120
    ) {materialFunction=/Game/Packages/jl_tests/combine8Samples}.Result
    False=MaterialFunctionCall!(
     Base=@n89.Color
     BaseSampleBias=@n108
     Scale=@n119
     Sample1=@n92.Color
     Sample2=@n93.Color
     Sample3=@n94.Color
     Sample4=@n95.Color
     Divisor=@n114
     UseCenteredLines=@n120
     UseRGB
    ) {materialFunction=/Game/Packages/jl_tests/combine4Samples}.Result
   ) {defaultValue=true, group=PerfOptions, parameterName=HighQulality8DepthSamples}) {r=true}
  NamedRerouteDeclaration(None=OneMinus(None=ComponentMask(None=@n121) {r=true})) {name=Depth_Outlines, nodeColor=<rgba 0.5234 1 0 1>}
  n122 = SceneTexture() {sceneTextureId=PPI_Roughness}
  n123 = ComponentMask(None=ScreenPosition().ViewportUV) {g=true, r=true}
  n124 = MaterialFunctionCall(OffsetAmount=NamedRerouteUsage()) {materialFunction=&BlurSampleOffsets}
  n125 = SceneTexture(UVs=Add(A=@n123, B=@n124."-.5,.5") {desc=x+}) {sceneTextureId=PPI_Roughness}
  n126 = SceneTexture(UVs=Add(A=@n123, B=@n124.".5,.5") {desc=y+}) {sceneTextureId=PPI_Roughness}
  n127 = Add(A=Abs(None=Subtract(A=@n122.Color, B=@n125.Color)), B=Abs(None=Subtract(A=@n122.Color, B=@n126.Color)))
  n128 = SceneTexture(UVs=Add(A=@n123, B=@n124.".5,-.5") {desc=x-}) {sceneTextureId=PPI_Roughness}
  n129 = SceneTexture(UVs=Add(A=@n123, B=@n124."-.5,-.5") {desc=y-}) {sceneTextureId=PPI_Roughness}
  n130 = Add(A=Abs(None=Subtract(A=@n122.Color, B=@n128.Color)), B=Abs(None=Subtract(A=@n122.Color, B=@n129.Color)))
  n131 = Saturate(None=Divide(A=Add(A=@n127, B=@n130), B=ScalarParameter() {group=RoughnessSampling, parameterName=Roughness_OutlineDivide}))
  NamedRerouteDeclaration(None=ComponentMask(None=OneMinus(None=@n131)) {r=true}) {name=Roughness_Outlines, nodeColor=<rgba 1 0 0.8437 1>}
  n132 = StaticSwitchParameter(True=@n53, False=@n59) {defaultValue=true, parameterName=DiagonalSampling}
  n133 = ScalarParameter() {defaultValue=4, group=NormalSampling, parameterName=Normal_Divisor}
  n134 = Divide(
   A=MaterialFunctionCall(X=@n132, Y=@n132) {materialFunction=&MakeFloat2}.Result
   B=MaterialFunctionCall(X=@n133, Y=Multiply(A=@n133) {constB=-1}) {materialFunction=&MakeFloat2}.Result
  )
  n135 = Multiply(
   A=OneMinus(None=Clamp(None=@n134) {desc=depth outline})
   B=ScalarParameter() {defaultValue=4, group=NormalSampling, parameterName=Normal_Scale}
  )
  n136 = MaterialFunctionCall(Float2=OneMinus(None=Clamp(None=@n135) {desc=depth outline})) {materialFunction=&BreakOutFloat2Components}
  n137 = Multiply(A=@n136.G)
  NamedRerouteDeclaration(None=Saturate(None=Max(A=@n136.R, B=@n137))) {name=Normal_Outlines, nodeColor=<rgba 0 1 0.8359 1>}
  n138 = SceneTexture() {sceneTextureId=PPI_StoredSpecular}
  n139 = Subtract(A=SceneDepth().R) {constB=-1}
  n140 = Multiply(A=Saturate(None=Subtract(A=@n139, B=@n31)), B=Saturate(None=Divide(A=@n31) {constB=0.01}))
  n141 = Add(A=Saturate(None=Ceil(None=@n138.Color)), B=@n140)
  NamedRerouteDeclaration(None=FeatureLevelSwitch(Default=Saturate(None=@n141), ES3_1=@n140)) {name=CustomDepth_PostMask, nodeColor=<rgba 1 0.5156 0 1>}
  n142 = SceneTexture() {sceneTextureId=PPI_PostProcessInput0}
  n143 = ComponentMask(None=ScreenPosition().ViewportUV) {g=true, r=true}
  n144 = MaterialFunctionCall(OffsetAmount=ScalarParameter() {defaultValue=0.1, parameterName=BlurOffset}) {materialFunction=&BlurSampleOffsets}
  n145 = SceneTexture(UVs=Add(A=@n143, B=@n144."0,1")) {sceneTextureId=PPI_PostProcessInput0}
  n146 = Subtract(A=DotProduct(A=@n142.Color, B=@n145.Color), B=NamedRerouteUsage())
  n147 = Floor(None=@n146)
  n148 = SceneTexture(UVs=Add(A=@n143, B=@n144."1,0")) {sceneTextureId=PPI_PostProcessInput0}
  n149 = Floor(None=Subtract(A=DotProduct(A=@n142.Color, B=@n148.Color), B=NamedRerouteUsage()))
  n150 = Add(
   A=LinearInterpolate(A=@n142.Color, B=@n145.Color, Alpha=Saturate(None=@n147))
   B=LinearInterpolate(A=@n142.Color, B=@n148.Color, Alpha=Saturate(None=@n149))
  )
  n151 = SceneTexture(UVs=Add(A=@n143, B=@n144."0,-1")) {sceneTextureId=PPI_PostProcessInput0}
  n152 = Floor(None=Subtract(A=DotProduct(A=@n142.Color, B=@n151.Color), B=NamedRerouteUsage()))
  n153 = SceneTexture(UVs=Add(A=@n144."-1,0", B=@n143)) {sceneTextureId=PPI_PostProcessInput0}
  n154 = Floor(None=Subtract(A=DotProduct(A=@n142.Color, B=@n153.Color), B=NamedRerouteUsage()))
  n155 = Add(
   A=LinearInterpolate(A=@n142.Color, B=@n151.Color, Alpha=Saturate(None=@n152))
   B=LinearInterpolate(A=@n142.Color, B=@n153.Color, Alpha=Saturate(None=@n154))
  )
  n156 = SceneTexture(UVs=Add(A=@n144."-.5,.5", B=@n143)) {sceneTextureId=PPI_PostProcessInput0}
  n157 = Subtract(A=DotProduct(A=@n142.Color, B=@n156.Color), B=NamedRerouteUsage())
  n158 = Floor(None=@n157)
  n159 = SceneTexture(UVs=Add(A=@n144.".5,.5", B=@n143)) {sceneTextureId=PPI_PostProcessInput0}
  n160 = Subtract(A=DotProduct(A=@n142.Color, B=@n159.Color), B=NamedRerouteUsage())
  n161 = Floor(None=@n160)
  n162 = Add(
   A=LinearInterpolate(A=@n142.Color, B=@n156.Color, Alpha=Saturate(None=@n158))
   B=LinearInterpolate(A=@n142.Color, B=@n159.Color, Alpha=Saturate(None=@n161))
  )
  n163 = SceneTexture(UVs=Add(A=@n144.".5,-.5", B=@n143)) {sceneTextureId=PPI_PostProcessInput0}
  n164 = Subtract(A=DotProduct(A=@n142.Color, B=@n163.Color), B=NamedRerouteUsage())
  n165 = Floor(None=@n164)
  n166 = SceneTexture(UVs=Add(A=@n144."-.5,-.5", B=@n143)) {sceneTextureId=PPI_PostProcessInput0}
  n167 = Subtract(A=DotProduct(A=@n142.Color, B=@n166.Color), B=NamedRerouteUsage())
  n168 = Floor(None=@n167)
  n169 = Add(
   A=LinearInterpolate(A=@n142.Color, B=@n163.Color, Alpha=Saturate(None=@n165))
   B=LinearInterpolate(A=@n142.Color, B=@n166.Color, Alpha=Saturate(None=@n168))
  )
  n170 = Add(A=@n142.Color, B=Add(A=Add(A=@n150, B=@n155), B=Add(A=@n162, B=@n169)))
  NamedRerouteDeclaration(None=ComponentMask(None=Divide(A=@n170) {constB=9}) {b=true, g=true, r=true}) {name=Blured_Scene, nodeColor=<rgba 0.1953 0 1 1>}
  n171 = ScalarParameter() {defaultValue=0.75, group=Fog, parameterName=FogMax}
  n172 = SceneDepth()
  n173 = ScalarParameter() {defaultValue=2000, group=Fog, parameterName=FogStartDistance}
  n174 = Power(
   Base=Divide(A=Subtract(A=@n172.R, B=@n173), B=ScalarParameter() {defaultValue=5000000, group=Fog, parameterName=FogDensity})
   Exp=ScalarParameter() {defaultValue=3, group=Fog, parameterName=FogFalloff}
  )
  NamedRerouteDeclaration(None=Saturate(None=Min(A=@n171, B=@n174))) {name=FogDistanceBlend, nodeColor=<rgba 0.125 1 0 1>}
  n175 = Divide(
   A=Subtract(A=SceneDepth().R, B=ScalarParameter() {defaultValue=2000, group=ColorTintOutlines, parameterName=ColorTint_StartDistance})
   B=ScalarParameter() {defaultValue=5000000, group=ColorTintOutlines, parameterName=ColorTint_Falloff}
  )
  NamedRerouteDeclaration(None=Multiply(A=ScalarParameter() {group=ColorTintOutlines, parameterName=ColorTintOutline}, B=Saturate(None=@n175))) {name=SceneColorDistanceBlend, nodeColor=<rgba 0.125 1 0 1>}
  NamedRerouteDeclaration(None=@n6.Debug) {name=Debug Value, nodeColor=<rgba 0 0.7656 1 1>}
  n176 = Constant()
  NamedRerouteDeclaration(None=If(
    A=ScalarParameter() {defaultValue=4, parameterName=PropMaskShadingModelID}
    B=SceneTexture() {sceneTextureId=PPI_ShadingModelID}.Color
    "A > B"=@n176
    "A == B"=Constant() {r=1}
    "A < B"=@n176
   )) {name=ClearCoat Shading Model Mask, nodeColor=<rgba 0.8516 1 0 1>}
  n177 = VectorParameter() {group=Color, parameterName=OutlineColor, defaultValue=<rgba 0 0 0 1>}
  n178 = NamedRerouteUsage()
  n179 = NamedRerouteUsage()
  n180 = SceneTexture() {sceneTextureId=PPI_PostProcessInput0}
  n181 = NamedRerouteUsage()
  n182 = NamedRerouteUsage()
  n183 = StaticSwitchParameter(True=LinearInterpolate(B=@n84, Alpha=NamedRerouteUsage()) {constA=1}, False=@n84) {defaultValue=true, parameterName=UseRoughnessFoliageMask}
  n184 = StaticSwitchParameter(True=NamedRerouteUsage(), False=FeatureLevelSwitch(Default=Multiply(A=@n182, B=@n183), ES3_1=NamedRerouteUsage())) {parameterName=DebugDepthOutlinesOnly}
  n185 = LinearInterpolate(
   A=StaticSwitchParameter(True=NamedRerouteUsage(), False=StaticSwitchParameter(True=@n181, False=@n184) {parameterName=DebugNormalOutlinesOnly}) {parameterName=DebugRoughnessOutlinesOnly}
   Alpha=NamedRerouteUsage()
  )
  n186 = Add(A=LinearInterpolate(A=Saturate(None=@n185), Alpha=NamedRerouteUsage()), B=NamedRerouteUsage())
  NamedRerouteDeclaration(None=LinearInterpolate(
    A=LinearInterpolate(A=@n177.RGB, B=@n178, Alpha=@n179)
    B=ComponentMask(None=@n180.Color) {b=true, g=true, r=true}
    Alpha=Saturate(None=@n186)
   ) {constB=0}) {name=Final Image With Outlines, nodeColor=<rgba 0.8281 0 1 1>}
  n187 = ScalarParameter() {defaultValue=1, group=DepthSampling, parameterName=BaseSampleBias_1}
  n188 = ScalarParameter() {defaultValue=1, group=DepthSampling_2, parameterName=BaseSampleBias_2}
  n189 = NamedRerouteUsage()
  NamedRerouteDeclaration(None=LinearInterpolate(
    A=LinearInterpolate(A=@n187, B=@n188, Alpha=@n189)
    B=ScalarParameter() {defaultValue=1, group=DepthSampling_3, parameterName=BaseSampleBias_3}
    Alpha=NamedRerouteUsage()
   )) {name=BaseSampleBias, nodeColor=<rgba 1 0 0.0234 1>}
  n190 = SceneTexture(UVs=@n61.ViewportUV) {sceneTextureId=PPI_Roughness}
  n191 = Constant()
  n192 = If(A=@n190.Color, B=NamedRerouteUsage(), "A > B"=@n191, "A == B"=@n191, "A < B"=Constant() {r=1}) {constB=0.01}
  n193 = Constant()
  n194 = If(A=@n190.Color, B=NamedRerouteUsage(), "A > B"=Constant() {r=1}, "A == B"=@n193, "A < B"=@n193) {constB=0.01}
  NamedRerouteDeclaration(None=Saturate(None=Add(A=@n192, B=@n194))) {name=Roughness ZERO Outline, nodeColor=<rgba 0 1 0.6484 1>}
  NamedRerouteDeclaration(None=ScalarParameter() {defaultValue=0.04, desc=RoughnessValue, parameterName=RoughnesThreshold_MIn}) {name=RoughMin, nodeColor=<rgba 1 0.8906 0 1>}
  NamedRerouteDeclaration(None=ScalarParameter() {defaultValue=0.5, desc=RoughnessValue, parameterName=RoughnesThreshold_max}) {name=RoughMax, nodeColor=<rgba 1 0.8906 0 1>}
  n195 = Subtract(
   A=SceneTexture() {sceneTextureId=PPI_SceneDepth}.Color
   B=ScalarParameter() {defaultValue=40000, group=SteppedTransition, parameterName=Step3_End}
  )
  n196 = ScalarParameter() {defaultValue=40000, group=SteppedTransition, parameterName=Step3_FadeDistance}
  NamedRerouteDeclaration(None=Saturate(None=Divide(A=@n195, B=@n196))) {name=Step3Blend, nodeColor=<rgba 0.4531 0 1 1>}
  NamedRerouteDeclaration(None=@n110) {name=debugTest, nodeColor=<rgba 0 1 0.7188 1>}
  NamedRerouteDeclaration(None=DotProduct(A=CameraVectorWS(), B=@n49.Color)) {name=FresnelMask, nodeColor=<rgba 1 0 0.7031 1>}
  n197 = LinearInterpolate(
   A=ScalarParameter() {defaultValue=-800, group=DepthSampling, parameterName=DivisorDepthSamples}
   B=ScalarParameter() {defaultValue=-800, group=DepthSampling_2, parameterName=DivisorDepthSamples2}
   Alpha=NamedRerouteUsage()
  )
  n198 = ScalarParameter() {defaultValue=-800, group=DepthSampling_3, parameterName=DivisorDepthSamples3}
  n199 = NamedRerouteUsage()
  NamedRerouteDeclaration(None=LinearInterpolate(
    A=LinearInterpolate(A=@n197, B=@n198, Alpha=@n199)
    B=ScalarParameter() {defaultValue=-800, group=DepthSampling_4, parameterName=DivisorDepthSamples4}
    Alpha=NamedRerouteUsage()
   )) {name=DivisoDepthSamples, nodeColor=<rgba 0 1 0.6953 1>}
  NamedRerouteDeclaration(None=MaterialFunctionCall(
    Color_Morning=VectorParameter() {parameterName=Fog Color_Morning, defaultValue=<rgba 0.1871 0.2343 0.2656 1>}.RGB
    Color_Day=VectorParameter() {parameterName=Fog Color_Day, defaultValue=<rgba 0.2259 0.3769 0.5417 1>}.RGB
    Color_Evening=VectorParameter() {parameterName=Fog Color_Evening, defaultValue=<rgba 0.1064 0.1776 0.2552 1>}.RGB
    Color_Night=VectorParameter() {parameterName=Fog Color_Night, defaultValue=<rgba 0.0688 0.1445 0.3385 1>}.RGB
   ) {materialFunction=&Color_Time_of_Day}.Result) {name=FogColor, nodeColor=<rgba 1 0 0.6797 1>}
  n200 = ScalarParameter() {defaultValue=1, group=DepthSampling, parameterName=LineThickness_1}
  n201 = ScalarParameter() {defaultValue=1, group=DepthSampling_2, parameterName=LineThickness_2}
  n202 = NamedRerouteUsage()
  NamedRerouteDeclaration(None=LinearInterpolate(
    A=LinearInterpolate(A=@n200, B=@n201, Alpha=@n202)
    B=ScalarParameter() {defaultValue=1, group=DepthSampling_3, parameterName=LineThickness_3}
    Alpha=NamedRerouteUsage()
   )) {name=LineThickness, nodeColor=<rgba 1 0.7734 0 1>}
  n203 = Subtract(
   A=SceneTexture() {sceneTextureId=PPI_SceneDepth}.Color
   B=ScalarParameter() {defaultValue=2000, group=SteppedTransition, parameterName=Step1_End}
  )
  n204 = ScalarParameter() {defaultValue=1, group=SteppedTransition, parameterName=Step1_FadeDistance}
  NamedRerouteDeclaration(None=Saturate(None=Divide(A=@n203, B=@n204))) {name=Step1Blend, nodeColor=<rgba 0.4531 0 1 1>}
  n205 = ScalarParameter() {defaultValue=2000, group=DepthSampling, parameterName=BiasLineThickness_1}
  n206 = ScalarParameter() {defaultValue=2000, group=DepthSampling_2, parameterName=BiasLineThickness_2}
  n207 = NamedRerouteUsage()
  NamedRerouteDeclaration(None=LinearInterpolate(
    A=LinearInterpolate(A=@n205, B=@n206, Alpha=@n207)
    B=ScalarParameter() {defaultValue=2000, group=DepthSampling_3, parameterName=BiasLineThickness_3}
    Alpha=NamedRerouteUsage()
   )) {name=BaseLineThickness, nodeColor=<rgba 1 0.4453 0 1>}
  n208 = Subtract(
   A=SceneTexture() {sceneTextureId=PPI_SceneDepth}.Color
   B=ScalarParameter() {defaultValue=40000, group=SteppedTransition, parameterName=Step2_End}
  )
  n209 = ScalarParameter() {defaultValue=40000, group=SteppedTransition, parameterName=Step2_FadeDistance}
  NamedRerouteDeclaration(None=Saturate(None=Divide(A=@n208, B=@n209))) {name=Step2Blend, nodeColor=<rgba 0.4531 0 1 1>}
  n210 = Subtract(
   A=ComponentMask(None=SceneTexture() {sceneTextureId=PPI_SceneDepth}.Color) {r=true}
   B=ScalarParameter() {defaultValue=500000, group=FadeOutlinesFarDistance, parameterName=FadeBiasOffset}
  )
  n211 = ScalarParameter() {defaultValue=20000, group=FadeOutlinesFarDistance, parameterName=FadeDist}
  NamedRerouteDeclaration(None=Saturate(None=Divide(A=@n210, B=@n211))) {name=DistanceFade, nodeColor=<rgba 0 1 1 1>}
  NamedRerouteUsage()
  OneMinus(None=Ceil(None=@n31))
  n212 = ScalarParameter() {defaultValue=9395040, group=CharacterMask, parameterName=CustomDepthMask}
  OneMinus(None=Saturate(None=Divide(A=@n31, B=@n212)))
  OneMinus()
  Round(None=@n146)
  Round(None=@n157)
  Round(None=@n160)
  Round(None=@n164)
  Round(None=@n167)
  ScalarParameter() {defaultValue=1, desc=Used to draw outlines on both sides of a discontinuity, group=DepthSampling, parameterName=CenteredLinesDepthSampling}
  n213 = ComponentMask(None=@n17.Color) {b=true, g=true, r=true}
  n214 = ScalarParameter() {defaultValue=5, parameterName=ColorQuantizeSteps}
  n215 = Round(None=Multiply(A=Length(None=@n213), B=@n214))
  StaticSwitchParameter(True=Multiply(A=Divide(A=@n215, B=@n214), B=SceneTexture() {sceneTextureId=PPI_DiffuseColor}.Color), False=@n17.Color) {parameterName=UseColorQuantize}
  StaticSwitchParameter(True=@n31, False=@n45) {defaultValue=true, parameterName=DiagonalSampling}
  ViewSize()
 ]
}
