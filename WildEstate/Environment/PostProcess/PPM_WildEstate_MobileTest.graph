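# Post-process material, named as a mobile test, that draws tinted outlines over scene color after depth of field.
# It samples scene color at the pixel and four diagonal offsets scaled by line thickness and texel size.
# The external combine4Samples function turns these into an edge mask from color rather than depth.
# The inverted mask, floored by an opacity parameter, apparently blends OutlineColor with the original image.
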Material /WildEstate/Environment/PostProcess/PPM_WildEstate_MobileTest
graph {
 properties {bAutomaticallySetUsageInEditor=false, blendableLocation=BL_SceneColorAfterDOF, materialDomain=MD_PostProcess, usedShadingModels=Default Lit}
 nodes [
  n1 = ComponentMask(None=ScreenPosition().ViewportUV) {g=true, r=true}
  n2 = ScalarParameter() {defaultValue=2, group=Posturizer, parameterName=MobilePosturizer_lineThickness}
  n3 = SceneTexelSize()
  n4 = MaterialFunctionCall!(OffsetAmount=NamedRerouteDeclaration(None=Multiply(A=@n2, B=@n3)) {name=LineThckenesMain, nodeColor=<rgba 1 0 0.7969 1>}) {materialFunction=/Engine/Functions/Engine_MaterialFunctions01/ImageAdjustment/BlurSampleOffsets}
  n5 = Add(A=@n1, B=@n4."-.5,.5") {desc=x+}
  n6 = Add(A=@n1, B=@n4.".5,.5") {desc=y+}
  n7 = Add(A=@n1, B=@n4.".5,-.5") {desc=x-}
  n8 = Add(A=@n1, B=@n4."-.5,-.5") {desc=y-}
  n9 = ScalarParameter() {defaultValue=1, desc=Used to draw outlines on both sides of a discontinuity, group=Posturizer, parameterName=CenteredLinesDepthSampling}
  n10 = ComponentMask(None=MaterialFunctionCall!(
    Base=SceneTexture() {sceneTextureId=PPI_PostProcessInput0}.Color
    BaseSampleBias=ScalarParameter() {group=Posturizer, parameterName=MobilePosturizer_Bias}
    Scale=ScalarParameter() {defaultValue=1.1374, group=Posturizer, parameterName=MobilePosturizer_Scale}
    Sample1=SceneTexture(UVs=@n5) {sceneTextureId=PPI_PostProcessInput0}.Color
    Sample2=SceneTexture(UVs=@n6) {sceneTextureId=PPI_PostProcessInput0}.Color
    Sample3=SceneTexture(UVs=@n7) {sceneTextureId=PPI_PostProcessInput0}.Color
    Sample4=SceneTexture(UVs=@n8) {sceneTextureId=PPI_PostProcessInput0}.Color
    Divisor=ScalarParameter() {defaultValue=0.5268, group=Posturizer, parameterName=MobilePosturizer_DIvisor}
    UseCenteredLines=@n9
    UseRGB=StaticBool() {value=true}
   ) {materialFunction=/Game/Packages/jl_tests/combine4Samples}.Result) {r=true}
  NamedRerouteDeclaration(None=OneMinus(None=ComponentMask(None=@n10) {r=true})) {name=Mobile_Outlines, nodeColor=<rgba 0.5234 1 0 1>}
  n11 = SceneTexture() {sceneTextureId=PPI_PostProcessInput0}
  n12 = NamedRerouteUsage()
  n13 = ScalarParameter() {defaultValue=0.572, group=Posturizer, parameterName=MobilePosturizer_OpacityOutline}
  NamedRerouteDeclaration(None=LinearInterpolate(
    A=VectorParameter() {group=Color, parameterName=OutlineColor, defaultValue=<rgba 0 0 0 1>}.RGB
    B=ComponentMask(None=@n11.Color) {b=true, g=true, r=true}
    Alpha=Max(A=@n12, B=@n13)
   ) {constB=0}) {name=Final Image With Outlines, nodeColor=<rgba 0.8281 0 1 1>}
  NamedRerouteUsage()
 ]
}
