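# Test post-process material for WildEstate, drawing stylized outlines and screen-space shadows after depth of field.
# Outlines come from four-tap depth, normal and roughness edge tests, with custom-depth/stencil exclusion, resolution-aware sampling and distance fading.
# Custom HLSL ray-marches depth against clear-coat shading-model pixels, using temporal dither to vary steps, then fades shadows with distance.
# A shading-path switch offers mobile posterizer outlines; Epic quality adds an extra multiply.
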Material /WildEstate/Environment/PostProcess/PPM_WildEstate_Test
graph {
 properties {bAutomaticallySetUsageInEditor=false, blendableLocation=BL_SceneColorAfterDOF, materialDomain=MD_PostProcess, usedShadingModels=Default Lit}
 nodes [
  n1 = NamedRerouteUsage()
  n2 = ComponentMask(None=SceneTexture() {sceneTextureId=PPI_PostProcessInput1}.Color) {b=true, g=true, r=true}
  n3 = Length(None=@n2) {is=class MaterialExpressionLength}
  n4 = Saturate(None=Multiply(A=ComponentMask(None=@n3) {r=true}, B=ScalarParameter() {defaultValue=50, parameterName=TranslucentMask}))
  n5 = ScalarParameter() {defaultValue=0.6, parameterName=TranslucentMask_min}
  Add(A=Multiply(A=ScalarParameter() {defaultValue=0.5, parameterName=SpecularMaskIntensity}, B=@n1), B=Min(A=@n4, B=@n5))
  Custom!(Tex, UV, TraceVec, NumSteps, bias, StepSizeMAX, MaxDepth) {
   code="// The below expression will get compiled\r\n// into the output of this node\r\n//float3(1, 1, 1);\r\n//float3 Color = SceneTextureLookup(ViewportUVToSceneTextureUV(UV, 1), 1, false);\r\n\r\nfloat TimeLerp = 1;\r\nfloat DepthDiff = 0;\r\nfloat DepthDiff2 = 0;\r\n\r\nfloat LastDiff = -bias;\r\nfloat StepSize = StepSizeMAX/NumSteps;\r\nDebug = 1;\r\nfloat3 RayStepUVz = float3(TraceVec.x, TraceVec.y, TraceVec.z);\r\n\r\nfloat accum = 0;\r\nfloat3 RayUVz = float3(UV, 0 );\r\nfloat3 RayUVzInv = float3(UV, 0 );\r\nfloat depth = SceneTextureLookupCommon(Parameters, ViewportUVToSceneTextureUV(UV, 1), 1, false);\r\nint i = 0;\r\nwhile (i < NumSteps)\r\n{\r\n\r\n\tRayUVz += RayStepUVz*StepSize;\r\n    RayUVzInv -= RayStepUVz*StepSize;\r\n    RayUVz.xy = saturate(RayUVz.xy);\r\n    RayUVzInv.xy = saturate(RayUVzInv.xy);\r\n\r\n    float MaskSample1 = SceneTextureLookupCommon(Parameters, ViewportUVToSceneTextureUV(RayUVz.xy, 23), 23, false);\r\n    if (MaskSample1 == 4)\r\n\t\t{\r\n\r\n        MaskSample1 = 1;\r\n        }\r\n    else\r\n\t\t{\r\n\r\n        MaskSample1 = 0;\r\n        }  \r\n    float MaskSample2 = SceneTextureLookupCommon(Parameters, ViewportUVToSceneTextureUV(RayUVzInv.xy, 23), 23, false);\r\n    if (MaskSample2 == 4)\r\n\t\t{\r\n\r\n        MaskSample2 = 1;\r\n        }\r\n    else\r\n\t\t{\r\n\r\n        MaskSample2 = 0;\r\n        }  \r\n\tfloat SampleDepth = SceneTextureLookupCommon(Parameters, ViewportUVToSceneTextureUV(RayUVz.xy, 1), 1, false);\r\n    float SampleDepth2 = SceneTextureLookupCommon(Parameters, ViewportUVToSceneTextureUV(RayUVzInv.xy, 1), 1, false);\r\n\t//DepthDiff = (SampleDepth+RayUVz.z)-depth + max((SampleDepth2-depth),0);\r\n   // DepthDiff = (SampleDepth+RayUVz.z)-depth + (SampleDepth2-depth);\r\n    DepthDiff = (SampleDepth+RayUVz.z)-depth ;\r\n    DepthDiff *= MaskSample1; \r\n    //DepthDiff = (SampleDepth+RayUVz.z)-depth;\r\n    //DepthDiff = (SampleDepth+RayUVz.z)-depth;\r\n    DepthDiff2 = (SampleDepth2+RayUVzInv.z)-depth;\r\n    DepthDiff2 *= MaskSample2; \r\n    \r\n\r\n\t\tif (DepthDiff < 0 &&&& DepthDiff > -MaxDepth)\r\n\t\t{\r\n\r\n\t\t\t//if(LastDiff > 0)\r\n\t\t\t//{\r\n\t\t\t//\tTimeLerp = saturate( LastDiff / (LastDiff - DepthDiff));\r\n\t\t\t//\taccum += StepSize * (1-TimeLerp);\r\n\t\t\t\t//accum+=StepSize;\r\n\t\t\t//}\r\n\t\t\t//else\r\n\t\t\t//{\r\n\t\t\t//\taccum+=StepSize;\r\n\t\t\t//}\r\n            accum += 0.1;\r\n\r\n\t\t}\r\n\t\telse\r\n\t\t//if(LastDiff <= 0)\r\n\t\t{\r\n\t\t\t//accum -=0.1;\r\n\t\t}\r\n //       if (DepthDiff2 < 0 )\r\n\t//\t{\r\n   //        accum +=0.1;\r\n        \r\n    //    }\r\n\r\n\t//LastDiff = DepthDiff;\r\n\r\n\ti++;\r\n}\r\n//Debug =  SceneTextureLookup(GetDefaultSceneTextureUV(Parameters, 23), 23, false);\r\nreturn 1-saturate(accum);\r\n"
   showCode=true
   additionalOutputs [{outputName=Debug, outputType=CMOT_Float1}]
   inputs [
    {inputName=Tex, input {expression=null, mask=1, maskA=1, maskB=1, maskG=1, maskR=1}}
    {inputName=UV, input {expression=null}}
    {inputName=TraceVec, input {expression=null}}
    {inputName=NumSteps, input {expression=null}}
    {inputName=bias, input {expression=null}}
    {inputName=StepSizeMAX, input {expression=null}}
    {inputName=MaxDepth, input {expression=null}}
   ]
  }
  Desaturation(None=@n2)
  n6 = ComponentMask(None=SceneTexture() {sceneTextureId=PPI_PostProcessInput0}.Color) {b=true, g=true, r=true}
  LinearInterpolate(A=Constant3Vector(), B=@n6, Alpha=Max(A=NamedRerouteUsage(), B=NamedRerouteUsage()))
  n7 = ComponentMask(None=SceneTexture(UVs=NamedRerouteUsage()) {sceneTextureId=PPI_Roughness}.Color) {r=true}
  n8 = ScalarParameter() {defaultValue=4, parameterName=SpecularMaskRoundingSteps}
  n9 = Multiply(A=@n7, B=@n8)
  LinearInterpolate(A=@n7, B=Divide(A=Round(None=@n9), B=@n8), Alpha=NamedRerouteUsage())
  n10 = ComponentMask(None=SceneTexture(UVs=NamedRerouteUsage()) {sceneTextureId=PPI_Roughness}.Color) {r=true}
  n11 = Multiply(A=@n8, B=@n10)
  LinearInterpolate(A=@n10, B=Divide(A=Round(None=@n11), B=@n8), Alpha=NamedRerouteUsage())
  n12 = ComponentMask(None=SceneTexture(UVs=NamedRerouteUsage()) {sceneTextureId=PPI_Roughness}.Color) {r=true}
  n13 = Multiply(A=@n8, B=@n12)
  LinearInterpolate(A=@n12, B=Divide(A=Round(None=@n13), B=@n8), Alpha=NamedRerouteUsage())
  n14 = ComponentMask(None=SceneTexture(UVs=NamedRerouteUsage()) {sceneTextureId=PPI_Roughness}.Color) {r=true}
  n15 = Multiply(A=@n8, B=@n14)
  LinearInterpolate(A=@n14, B=Divide(A=Round(None=@n15), B=@n8), Alpha=NamedRerouteUsage())
  Max(A=@n4, B=@n5)
  n16 = SceneTexture() {sceneTextureId=PPI_Opacity}
  Multiply(A=@n4, B=OneMinus(None=Ceil(None=@n16.Color)))
  n17 = NamedRerouteUsage()
  n18 = NamedRerouteUsage()
  n19 = Ceil(None=SceneTexture(UVs=NamedRerouteUsage()) {sceneTextureId=PPI_Specular}.Color)
  n20 = LinearInterpolate(A=@n17, B=@n18, Alpha=@n19)
  n21 = Ceil(None=SceneTexture(UVs=NamedRerouteUsage()) {sceneTextureId=PPI_Specular}.Color)
  n22 = LinearInterpolate(A=@n18, B=@n17, Alpha=@n21)
  n23 = NamedRerouteUsage()
  n24 = NamedRerouteUsage()
  n25 = Ceil(None=SceneTexture(UVs=NamedRerouteUsage()) {sceneTextureId=PPI_Specular}.Color)
  n26 = LinearInterpolate(A=@n23, B=@n24, Alpha=@n25)
  n27 = Ceil(None=SceneTexture(UVs=NamedRerouteUsage()) {sceneTextureId=PPI_Specular}.Color)
  n28 = LinearInterpolate(A=@n24, B=@n23, Alpha=@n27)
  n29 = Add(A=Add(A=Subtract(A=@n20, B=@n17), B=Subtract(A=@n22, B=@n18)), B=Add(A=Subtract(A=@n26, B=@n23), B=Subtract(A=@n28, B=@n24)))
  n30 = Multiply(A=NamedRerouteUsage())
  Multiply(A=Saturate(None=Step(Y=@n29, X=@n30)), B=Max(A=Max(A=@n19, B=@n21), B=Max(A=@n25, B=@n27)))
  n31 = Multiply(A=ComponentMask(None=ViewSize()) {g=true}, B=ViewProperty() {property=MEVP_ResolutionFraction}.Property)
  n32 = ScalarParameter() {defaultValue=1081, parameterName=VerticalResolutionThreshold}
  NamedRerouteDeclaration(None=Saturate(None=Subtract(A=@n31, B=@n32))) {name=Is High Res, nodeColor=<rgba 1 0 0.1406 1>}
  n33 = ComponentMask(None=@n7) {r=true}
  n34 = ComponentMask(None=@n10) {r=true}
  n35 = ComponentMask(None=@n12) {r=true}
  n36 = ComponentMask(None=@n14) {r=true}
  n37 = Max(A=Abs(None=Subtract(A=@n33, B=@n34)), B=Abs(None=Subtract(A=@n35, B=@n36)))
  n38 = ScalarParameter() {defaultValue=0.01, group=Outlines Roughness, parameterName=RoughnessThreshold, sliderMax=30, sortPriority=1}
  n39 = ScalarParameter() {defaultValue=0.03, group=Outlines Roughness, parameterName=RoughnessThresholdSoft_Width, sliderMax=1, sortPriority=2}
  n40 = NamedRerouteUsage()
  n41 = MaterialFunctionCall(Input=@n37, Input Low=@n38, Input High=Add(A=@n38, B=Multiply(A=@n39, B=@n40)) {constB=0.03}) {materialFunction=&RemapValue_0-1}
  NamedRerouteDeclaration(None=StaticSwitchParameter(True=OneMinus(None=@n41.Result), False=Step(Y=@n37, X=@n38)) {group=Outlines Roughness, parameterName=UseSoftGraduatingRougnessThreshold}) {name=Roughness Outlines, nodeColor=<rgba 0 0.25 1 1>}
  n42 = ComponentMask(None=ScreenPosition().ViewportUV) {g=true, r=true}
  n43 = ScalarParameter() {defaultValue=2, group=Posturizer, parameterName=MobilePosturizer_lineThickness}
  n44 = SceneTexelSize()
  n45 = MaterialFunctionCall!(OffsetAmount=NamedRerouteDeclaration(None=Multiply(A=@n43, B=@n44)) {name=LineThckenesMain, nodeColor=<rgba 1 0 0.7969 1>}) {materialFunction=/Engine/Functions/Engine_MaterialFunctions01/ImageAdjustment/BlurSampleOffsets}
  n46 = Add(A=@n42, B=@n45."-.5,.5") {desc=x+}
  n47 = Add(A=@n42, B=@n45.".5,.5") {desc=y+}
  n48 = Add(A=@n42, B=@n45.".5,-.5") {desc=x-}
  n49 = Add(A=@n42, B=@n45."-.5,-.5") {desc=y-}
  n50 = ScalarParameter() {defaultValue=1, desc=Used to draw outlines on both sides of a discontinuity, group=Posturizer, parameterName=CenteredLinesDepthSampling}
  n51 = ComponentMask(None=MaterialFunctionCall!(
    Base=SceneTexture() {sceneTextureId=PPI_PostProcessInput0}.Color
    BaseSampleBias=ScalarParameter() {group=Posturizer, parameterName=MobilePosturizer_Bias}
    Scale=ScalarParameter() {defaultValue=1.1374, group=Posturizer, parameterName=MobilePosturizer_Scale}
    Sample1=SceneTexture(UVs=@n46) {sceneTextureId=PPI_PostProcessInput0}.Color
    Sample2=SceneTexture(UVs=@n47) {sceneTextureId=PPI_PostProcessInput0}.Color
    Sample3=SceneTexture(UVs=@n48) {sceneTextureId=PPI_PostProcessInput0}.Color
    Sample4=SceneTexture(UVs=@n49) {sceneTextureId=PPI_PostProcessInput0}.Color
    Divisor=ScalarParameter() {defaultValue=0.5268, group=Posturizer, parameterName=MobilePosturizer_DIvisor}
    UseCenteredLines=@n50
    UseRGB=StaticBool() {value=true}
   ) {materialFunction=/Game/Packages/jl_tests/combine4Samples}.Result) {r=true}
  NamedRerouteDeclaration(None=OneMinus(None=ComponentMask(None=@n51) {r=true})) {name=Mobile_Outlines, nodeColor=<rgba 0.5234 1 0 1>}
  n52 = Constant()
  NamedRerouteDeclaration(None=If(
    A=ScalarParameter() {defaultValue=4, group=Screen Space Shadows, parameterName=PropMaskShadingModelID}
    B=SceneTexture() {sceneTextureId=PPI_ShadingModelID}.Color
    "A > B"=@n52
    "A == B"=Constant() {r=1}
    "A < B"=@n52
   )) {name=ClearCoat Shading Model Mask, nodeColor=<rgba 0.8516 1 0 1>}
  n53 = NamedRerouteUsage()
  MaterialExpressionNamedRerouteUsage_69 = NamedRerouteUsage()
  MaterialExpressionNamedRerouteUsage_68 = NamedRerouteUsage()
  MaterialExpressionNamedRerouteUsage_67 = NamedRerouteUsage()
  MaterialExpressionNamedRerouteUsage_70 = NamedRerouteUsage()
  MaterialExpressionNamedRerouteUsage_71 = NamedRerouteUsage()
  MaterialExpressionNamedRerouteUsage_66 = NamedRerouteUsage()
  MaterialExpressionNamedRerouteUsage_72 = NamedRerouteUsage()
  MaterialExpressionNamedRerouteUsage_76 = NamedRerouteUsage()
  n54 = Custom!(
   Tex=@MaterialExpressionNamedRerouteUsage_69
   UV=@MaterialExpressionNamedRerouteUsage_68
   TraceVec=@MaterialExpressionNamedRerouteUsage_67
   NumSteps=@MaterialExpressionNamedRerouteUsage_70
   bias=@MaterialExpressionNamedRerouteUsage_71
   StepSizeMAX=@MaterialExpressionNamedRerouteUsage_66
   MaxDepth=@MaterialExpressionNamedRerouteUsage_72
   MinDepth=@MaterialExpressionNamedRerouteUsage_76
  ) {
   code="// The below expression will get compiled\r\n// into the output of this node\r\n//float3(1, 1, 1);\r\n//float3 Color = SceneTextureLookup(ViewportUVToSceneTextureUV(UV, 1), 1, false);\r\n\r\nfloat TimeLerp = 1;\r\nfloat DepthDiff = 0;\r\n//float DepthDiff2 = 0;\r\n\r\nfloat LastDiff = -bias;\r\nfloat StepSize = StepSizeMAX/NumSteps;\r\nDebug = 1;\r\nfloat3 RayStepUVz = float3(TraceVec.x, TraceVec.y, TraceVec.z);\r\n\r\nfloat accum = 0;\r\nfloat3 RayUVz = float3(UV, 0 );\r\n//float3 RayUVzInv = float3(UV, 0 );\r\nfloat depth = SceneTextureLookupCommon(Parameters, ViewportUVToSceneTextureUV(UV, 1), 1, false);\r\nint i = 0;\r\nwhile (i < NumSteps)\r\n{\r\n\r\n\tRayUVz += RayStepUVz*StepSize;\r\n    //RayUVzInv -= RayStepUVz*StepSize;\r\n    RayUVz.xy = saturate(RayUVz.xy);\r\n    //RayUVzInv.xy = saturate(RayUVzInv.xy);\r\n    float CustomD = SceneTextureLookupCommon(Parameters, ViewportUVToSceneTextureUV(RayUVz.xy, 13), 13, false);\r\n    CustomD -= MinDepth; \r\n    CustomD = 1-saturate(CustomD);\r\n    if(CustomD == 0)\r\n{\t\r\n    float MaskSample1 = SceneTextureLookupCommon(Parameters, ViewportUVToSceneTextureUV(RayUVz.xy, 23), 23, false);\r\n    if (MaskSample1 == 4)\r\n\t\t{\r\n\r\n        MaskSample1 = 1;\r\n        }\r\n    else\r\n\t\t{\r\n\r\n        MaskSample1 = 0;\r\n        i++;\r\n        continue;\r\n\r\n        }  \r\n   /* float MaskSample2 = SceneTextureLookupCommon(Parameters, ViewportUVToSceneTextureUV(RayUVzInv.xy, 23), 23, false);\r\n    if (MaskSample2 == 4)\r\n\t\t{\r\n\r\n        MaskSample2 = 1;\r\n\r\n        }\r\n    else\r\n\t\t{\r\n\r\n        MaskSample2 = 0;\r\n\r\n        }  \r\n*/ \tfloat SampleDepth = SceneTextureLookupCommon(Parameters, ViewportUVToSceneTextureUV(RayUVz.xy, 1), 1, false);\r\n    //float SampleDepth2 = SceneTextureLookupCommon(Parameters, ViewportUVToSceneTextureUV(RayUVzInv.xy, 1), 1, false);\r\n\t//DepthDiff = (SampleDepth+RayUVz.z)-depth + max((SampleDepth2-depth),0);\r\n   // DepthDiff = (SampleDepth+RayUVz.z)-depth + (SampleDepth2-depth);\r\n    DepthDiff = (SampleDepth+RayUVz.z)-depth ;\r\n    DepthDiff *= MaskSample1; \r\n    //DepthDiff = (SampleDepth+RayUVz.z)-depth;\r\n    //DepthDiff = (SampleDepth+RayUVz.z)-depth;\r\n    //DepthDiff2 = (SampleDepth2+RayUVzInv.z)-depth;\r\n    //DepthDiff2 *= MaskSample2; \r\n    \r\n\r\n\t\tif (DepthDiff < 0 &&&& DepthDiff > -MaxDepth)\r\n\t\t{\r\n\r\n\t\t\t//if(LastDiff > 0)\r\n\t\t\t//{\r\n\t\t\t//\tTimeLerp = saturate( LastDiff / (LastDiff - DepthDiff));\r\n\t\t\t//\taccum += StepSize * (1-TimeLerp);\r\n\t\t\t\t//accum+=StepSize;\r\n\t\t\t//}\r\n\t\t\t//else\r\n\t\t\t//{\r\n\t\t\t//\taccum+=StepSize;\r\n\t\t\t//}\r\n            accum += 0.1;\r\n\r\n\t\t}\r\n\t\telse\r\n\t\t//if(LastDiff <= 0)\r\n\t\t{\r\n\t\t\t//accum -=0.1;\r\n\t\t}\r\n //       if (DepthDiff2 < 0 )\r\n\t//\t{\r\n   //        accum +=0.1;\r\n        \r\n    //    }\r\n\r\n\t//LastDiff = DepthDiff;\r\n\r\n\ti++;\r\n}\r\nelse{\r\n i++;\r\n}\r\n}\r\n//Debug =  SceneTextureLookup(GetDefaultSceneTextureUV(Parameters, 23), 23, false);\r\nreturn 1-saturate(accum);\r\n"
   desc=more expensive but with custom depth masking
   showCode=true
   additionalOutputs [{outputName=Debug, outputType=CMOT_Float1}]
   inputs [
    {inputName=Tex, input {expression=&self:MaterialExpressionNamedRerouteUsage_69}}
    {inputName=UV, input {expression=&self:MaterialExpressionNamedRerouteUsage_68}}
    {inputName=TraceVec, input {expression=&self:MaterialExpressionNamedRerouteUsage_67}}
    {inputName=NumSteps, input {expression=&self:MaterialExpressionNamedRerouteUsage_70}}
    {inputName=bias, input {expression=&self:MaterialExpressionNamedRerouteUsage_71}}
    {inputName=StepSizeMAX, input {expression=&self:MaterialExpressionNamedRerouteUsage_66}}
    {inputName=MaxDepth, input {expression=&self:MaterialExpressionNamedRerouteUsage_72}}
    {inputName=MinDepth, input {expression=&self:MaterialExpressionNamedRerouteUsage_76}}
   ]
  }
  n55 = Round(None=@n54.return)
  n56 = ScalarParameter() {defaultValue=0.75, group=Screen Space Shadows, parameterName=Ray_ShadowMin}
  n57 = SceneTexture() {sceneTextureId=PPI_SceneDepth}
  n58 = ComponentMask(None=@n57.Color) {r=true}
  n59 = ScalarParameter() {defaultValue=1000, group=Screen Space Shadows, parameterName=DepthFadeRaysOffset}
  n60 = Saturate(None=Divide(A=Subtract(A=@n58, B=@n59), B=ScalarParameter() {defaultValue=2000, group=Screen Space Shadows, parameterName=DepthFadeRays}))
  NamedRerouteDeclaration(None=Clamp(None=Add(A=@n53, B=@n55), Min=Add(A=@n56, B=@n60))) {name=ScreenSpaceShadows, nodeColor=<rgba 1 0 0.7969 1>}
  n61 = ComponentMask(None=SceneTexture(UVs=NamedRerouteUsage()) {sceneTextureId=PPI_PostProcessInput0}.Color) {b=true, g=true, r=true}
  n62 = ComponentMask(None=SceneTexture(UVs=NamedRerouteUsage()) {sceneTextureId=PPI_PostProcessInput0}.Color) {b=true, g=true, r=true}
  n63 = ComponentMask(None=SceneTexture(UVs=NamedRerouteUsage()) {sceneTextureId=PPI_PostProcessInput0}.Color) {b=true, g=true, r=true}
  n64 = ComponentMask(None=SceneTexture(UVs=NamedRerouteUsage()) {sceneTextureId=PPI_PostProcessInput0}.Color) {b=true, g=true, r=true}
  NamedRerouteDeclaration(None=Min(A=Min(A=@n61, B=@n62), B=Min(A=@n63, B=@n64))) {name=Final Color Min, nodeColor=<rgba 0.8281 1 0 1>}
  n65 = Max(A=Max(A=@n33, B=@n34), B=Max(A=@n35, B=@n36))
  NamedRerouteDeclaration(None=OneMinus(None=Round(None=@n65))) {name=Cluster Mask, nodeColor=<rgba 0.6172 0 1 1>}
  n66 = NamedRerouteUsage()
  n67 = ScalarParameter() {group=Outlines Opacity, parameterName=Opacity Start}
  n68 = Saturate(None=Divide(A=Subtract(A=@n66, B=@n67) {constB=20000}, B=ScalarParameter() {defaultValue=16000, group=Outlines Opacity, parameterName=Opacity Falloff}) {constB=2000})
  NamedRerouteDeclaration(None=LinearInterpolate(
    A=ScalarParameter() {defaultValue=0.22, group=Outlines Opacity, parameterName=Maximum Opacity}
    B=ScalarParameter() {defaultValue=0.75, group=Outlines Opacity, parameterName=Minimum Opacity}
    Alpha=NamedRerouteDeclaration(None=@n68) {name=Distance Mask, nodeColor=<rgba 1 0.2109 0 1>}
   )) {name=Distance Fade, nodeColor=<rgba 1 0.4219 0 1>}
  n69 = ScalarParameter() {defaultValue=0.5, group=Outlines Normal, parameterName=NormalThreshold, sliderMax=30, sortPriority=1}
  n70 = NamedRerouteDeclaration(None=SceneTexture(UVs=NamedRerouteUsage()) {sceneTextureId=PPI_WorldNormal}.Color) {name=Normal 1, nodeColor=<rgba 0.4568 0 1 1>}
  n71 = NamedRerouteDeclaration(None=SceneTexture(UVs=NamedRerouteUsage()) {sceneTextureId=PPI_WorldNormal}.Color) {name=Normal 2, nodeColor=<rgba 0.4568 0 1 1>}
  n72 = NamedRerouteDeclaration(None=SceneTexture(UVs=NamedRerouteUsage()) {sceneTextureId=PPI_WorldNormal}.Color) {name=Normal 3, nodeColor=<rgba 0.4568 0 1 1>}
  n73 = NamedRerouteDeclaration(None=SceneTexture(UVs=NamedRerouteUsage()) {sceneTextureId=PPI_WorldNormal}.Color) {name=Normal 4, nodeColor=<rgba 0.4568 0 1 1>}
  n74 = Min(A=Saturate(None=DotProduct(A=@n70, B=@n71)), B=Saturate(None=DotProduct(A=@n72, B=@n73)))
  NamedRerouteDeclaration(None=Max(A=Step(Y=@n69, X=@n74), B=NamedRerouteUsage())) {name=Normal Outlines, nodeColor=<rgba 0.5703 1 0 1>}
  n75 = ScalarParameter() {defaultValue=600000, group=Outlines Opacity, parameterName=Distant Cutoff}
  NamedRerouteDeclaration(None=Saturate(None=Subtract(A=@n66, B=@n75) {constB=20000})) {name=Distance Cutoff, nodeColor=<rgba 0.8047 0 1 1>}
  n76 = ComponentMask(None=SceneTexture(UVs=NamedRerouteUsage()) {sceneTextureId=PPI_CustomDepth}.Color) {r=true}
  n77 = ComponentMask(None=SceneTexture(UVs=NamedRerouteUsage()) {sceneTextureId=PPI_CustomDepth}.Color) {r=true}
  n78 = ComponentMask(None=SceneTexture(UVs=NamedRerouteUsage()) {sceneTextureId=PPI_CustomDepth}.Color) {r=true}
  n79 = ComponentMask(None=SceneTexture(UVs=NamedRerouteUsage()) {sceneTextureId=PPI_CustomDepth}.Color) {r=true}
  n80 = Subtract(A=Min(A=Min(A=@n76, B=@n77), B=Min(A=@n78, B=@n79)), B=NamedRerouteUsage())
  NamedRerouteDeclaration(None=OneMinus(None=Saturate(None=@n80))) {name=No Outlines Mask, nodeColor=<rgba 0 0.25 1 1>}
  n81 = ScalarParameter() {defaultValue=1, parameterName=Mobile MinDepth Opacity}
  n82 = ComponentMask(None=SceneTexture() {sceneTextureId=PPI_SceneDepth}.Color) {r=true}
  n83 = ScalarParameter() {defaultValue=50000, parameterName=MobileDepthOffset}
  n84 = Saturate(None=Divide(A=Subtract(A=@n82, B=@n83), B=ScalarParameter() {defaultValue=50000, parameterName=MobileDepthBlendDist}))
  NamedRerouteDeclaration(None=Saturate(None=LinearInterpolate(B=@n81, Alpha=@n84))) {name=Mobile DepthMask, nodeColor=<rgba 0 0.2969 1 1>}
  n85 = NamedRerouteUsage()
  n86 = NamedRerouteDeclaration(None=ComponentMask(None=SceneTexture(UVs=@n85) {sceneTextureId=PPI_SceneDepth}.Color) {r=true}) {name=Depth 1, nodeColor=<rgba 1 1 1 1>}
  n87 = NamedRerouteUsage()
  n88 = NamedRerouteDeclaration(None=ComponentMask(None=SceneTexture(UVs=@n87) {sceneTextureId=PPI_SceneDepth}.Color) {r=true}) {name=Depth 2, nodeColor=<rgba 1 1 1 1>}
  n89 = Abs(None=Subtract(A=@n86, B=@n88))
  n90 = NamedRerouteUsage()
  n91 = NamedRerouteDeclaration(None=ComponentMask(None=SceneTexture(UVs=@n90) {sceneTextureId=PPI_SceneDepth}.Color) {r=true}) {name=Depth 3, nodeColor=<rgba 1 1 1 1>}
  n92 = NamedRerouteUsage()
  n93 = NamedRerouteDeclaration(None=ComponentMask(None=SceneTexture(UVs=@n92) {sceneTextureId=PPI_SceneDepth}.Color) {r=true}) {name=Depth 4, nodeColor=<rgba 1 1 1 1>}
  n94 = Abs(None=Subtract(A=@n91, B=@n93))
  n95 = Min(A=@n86, B=@n88)
  n96 = Min(A=@n91, B=@n93)
  n97 = Multiply(
   A=LinearInterpolate(
    A=ScalarParameter() {defaultValue=30, group=Outlines Depth, parameterName=DepthThreshold, sliderMax=30, sortPriority=1}
    B=ScalarParameter() {defaultValue=200, group=Outlines Depth, parameterName=DepthThreshold Vegetation, sliderMax=30, sortPriority=1}
    Alpha=NamedRerouteUsage()
   )
   B=ComponentMask(None=SceneTexelSize()) {g=true}
  )
  n98 = Add(A=NamedRerouteUsage())
  n99 = Multiply(
   A=NamedRerouteDeclaration(None=Min(A=@n95, B=@n96)) {name=Min Depth, nodeColor=<rgba 1 0 0.7734 1>}
   B=Add(A=Multiply(A=@n97, B=@n98), B=NamedRerouteUsage())
  )
  NamedRerouteDeclaration(None=Step(Y=Max(A=@n89, B=@n94), X=NamedRerouteDeclaration(None=@n99) {name=Depth Threshold, nodeColor=<rgba 1 0.3516 0 1>})) {name=Depth Outlines, nodeColor=<rgba 1 0.25 1 1>}
  n100 = ScalarParameter() {defaultValue=0.0761, group=Screen Space Shadows, parameterName=Ray_StepSize_Low}
  n101 = ScalarParameter() {defaultValue=0.0761, group=Screen Space Shadows, parameterName=Ray_StepSize}
  n102 = MaterialFunctionCall!(Alpha Threshold, Random) {materialFunction=/Game/TimeOfDay/CloudsAndStorm/Materials/MaterialFunctions/DitherTemporalAABlue}
  NamedRerouteDeclaration(None=Multiply(A=OneMinus(None=@n60), B=LinearInterpolate(A=@n100, B=@n101, Alpha=@n102.Result))) {name=StepSizeMax, nodeColor=<rgba 0.7813 0 1 1>}
  n103 = Min(A=NamedRerouteUsage(), B=Min(A=NamedRerouteUsage(), B=NamedRerouteUsage()))
  n104 = Saturate(None=Add(A=@n1, B=Saturate(None=@n103)))
  n105 = MaterialFunctionCall!(ID=Constant() {r=6}) {
   desc=adding because items want outline but need custom depth
   materialFunction=/Game/Packages/Fortress_SharedMaterials/Stencil_ID_Definitions/StencilMaskCompare
  }
  n106 = Multiply(A=NamedRerouteUsage(), B=OneMinus(None=@n105.Result))
  NamedRerouteDeclaration(None=Max(A=Max(A=@n104, B=@n106), B=NamedRerouteUsage())) {name=FinalOutline, nodeColor=<rgba 0 1 0.0625 1>}
  NamedRerouteDeclaration(None=Normalize(VectorInput=VectorParameter() {group=Screen Space Shadows, parameterName=Ray_Dir, defaultValue=<rgba 1.1808 -1.2688 6 1>}.RGB)) {name=TraceVector, nodeColor=<rgba 0.7812 0 1 1>}
  NamedRerouteDeclaration(None=ScreenPosition().ViewportUV) {name=ScreenUVs, nodeColor=<rgba 0.7812 0 1 1>}
  NamedRerouteDeclaration(None=@n57.Color) {name=Tex, nodeColor=<rgba 0.7812 0 1 1>}
  NamedRerouteDeclaration(None=LinearInterpolate(
    A=ScalarParameter() {defaultValue=2, group=Screen Space Shadows, parameterName=Ray_Steps_Low}
    B=ScalarParameter() {defaultValue=16, group=Screen Space Shadows, parameterName=Ray_Steps}
    Alpha=@n102.Result
   )) {name=NumSteps, nodeColor=<rgba 0.7812 0 1 1>}
  NamedRerouteDeclaration(None=ScalarParameter() {defaultValue=0.01, group=Screen Space Shadows, parameterName=Ray_Bias}) {name=StepBias, nodeColor=<rgba 0.7812 0 1 1>}
  NamedRerouteDeclaration(None=ScalarParameter() {defaultValue=300, group=Screen Space Shadows, parameterName=Ray_MaxDepth}) {name=RayMaxDepth, nodeColor=<rgba 0.7812 0 1 1>}
  n107 = ComponentMask(None=SceneTexture() {sceneTextureId=PPI_SpecularColor}.Color) {r=true}
  NamedRerouteDeclaration(None=Saturate(None=Ceil(None=@n107))) {name=Specular Mask, nodeColor=<rgba 1 0.1641 0 1>}
  n108 = ComponentMask(None=SceneTexture(UVs=NamedRerouteUsage()) {sceneTextureId=PPI_PostProcessInput0}.Color) {b=true, g=true, r=true}
  n109 = ComponentMask(None=SceneTexture(UVs=NamedRerouteUsage()) {sceneTextureId=PPI_PostProcessInput0}.Color) {b=true, g=true, r=true}
  n110 = ComponentMask(None=SceneTexture(UVs=NamedRerouteUsage()) {sceneTextureId=PPI_PostProcessInput0}.Color) {b=true, g=true, r=true}
  n111 = ComponentMask(None=SceneTexture(UVs=NamedRerouteUsage()) {sceneTextureId=PPI_PostProcessInput0}.Color) {b=true, g=true, r=true}
  NamedRerouteDeclaration(None=Min(A=Min(A=@n108, B=@n109), B=Min(A=@n110, B=@n111))) {name=Posterized Min Color, nodeColor=<rgba 0.3125 0 1 1>}
  n112 = ComponentMask(None=ScreenPosition().ViewportUV) {g=true, r=true}
  n113 = ComponentMask(None=ScreenPosition().ViewportUV) {g=true, r=true}
  n114 = Constant()
  n115 = Multiply(A=SceneTexelSize(), B=ScalarParameter() {defaultValue=1, group=LineThicknessOveride for Capture, parameterName=LineThicknessOveride})
  n116 = ComponentMask(None=@n115) {g=true}
  n117 = AppendVector(A=@n114, B=@n116)
  NamedRerouteDeclaration(None=LinearInterpolate(A=@n112, B=Add(A=@n113, B=@n117), Alpha=NamedRerouteUsage())) {name=UV1, nodeColor=<rgba 1 0 0.3763 1>}
  n118 = Subtract(A=NamedRerouteDeclaration(None=@n108) {name=First Color Sample, nodeColor=<rgba 0 0.4141 1 1>}, B=@n109)
  n119 = Subtract(A=@n110, B=@n111)
  MaterialExpressionSaturate_17 = Saturate(None=Max(A=Abs(None=@n118), B=Abs(None=@n119)))
  n120 = Convert(None=@MaterialExpressionSaturate_17) {
   nodeName=Break Float3
   convertInputs [{
     type=Vector3
     expressionInput {
      expression=&self:MaterialExpressionSaturate_17
      inputName=None
      mask=0
      maskA=0
      maskB=0
      maskG=0
      maskR=0
      outputIndex=0
     }
    }]
   convertOutputs [{type=Scalar}, {type=Scalar}, {type=Scalar}]
  }
  n121 = OneMinus(None=Max(A=Max(A=@n120, B=@n120), B=@n120))
  n122 = ScalarParameter() {defaultValue=32, parameterName=Posterized Power}
  n123 = Saturate(None=Multiply(A=Power(Base=@n121, Exp=@n122) {constExponent=12}) {constB=1.1})
  NamedRerouteDeclaration(None=@n123) {name=Posterized Lines, nodeColor=<rgba 0.8281 1 0 1>}
  n124 = Add(A=Min(A=NamedRerouteUsage(), B=NamedRerouteUsage()), B=Add(A=NamedRerouteUsage(), B=NamedRerouteUsage()))
  n125 = ScalarParameter() {defaultValue=0.53, parameterName=Posterized Far Opacity}
  n126 = ScalarParameter() {defaultValue=0.8, parameterName=Posterized Close Opacity}
  n127 = NamedRerouteUsage()
  n128 = Saturate(None=Add(A=@n124, B=LinearInterpolate(A=@n125, B=@n126, Alpha=@n127) {constA=0.1, constB=0.6}))
  n129 = NamedRerouteUsage()
  NamedRerouteDeclaration(None=LinearInterpolate(A=Multiply(A=@n128, B=@n129), B=NamedRerouteUsage(), Alpha=Saturate(None=@n124)) {constB=0}) {name=Mobile Posturizer V2, nodeColor=<rgba 1 0.0293 0.6616 1>}
  n130 = ComponentMask(None=SceneTexture() {sceneTextureId=PPI_PostProcessInput0}.Color) {b=true, g=true, r=true}
  n131 = ScalarParameter() {defaultValue=-0.1, parameterName=Mobile_LuminanceBias_Low}
  n132 = ScalarParameter() {defaultValue=2, parameterName=Mobile_LuminanceBias_high}
  n133 = Desaturation(None=@n130)
  n134 = NamedRerouteUsage()
  n135 = ScalarParameter() {defaultValue=0.572, group=Posturizer, parameterName=MobilePosturizer_OpacityOutline}
  n136 = Max(A=Saturate(None=LinearInterpolate(A=@n131, B=@n132, Alpha=@n133)), B=Max(A=Max(A=@n134, B=@n135), B=NamedRerouteUsage()))
  n137 = NamedRerouteUsage()
  NamedRerouteDeclaration(None=LinearInterpolate(A=Constant4Vector() {constant=<rgba 0 0 0 1>}.RGB, B=@n130, Alpha=Max(A=@n136, B=@n137)) {constB=0}) {name=Mobile Posturizer V1, nodeColor=<rgba 0 0.1094 1 1>}
  n138 = Abs(None=Subtract(A=NamedRerouteUsage(), B=NamedRerouteUsage()))
  n139 = Abs(None=Subtract(A=NamedRerouteUsage(), B=NamedRerouteUsage()))
  n140 = NamedRerouteUsage()
  n141 = SceneTexelSize()
  n142 = Multiply(
   A=Multiply(
    A=ScalarParameter() {defaultValue=200, group=Outlines Depth, parameterName=MobileDepthThreshold, sliderMax=30, sortPriority=1}
    B=ComponentMask(None=@n141) {g=true}
   )
   B=Add(A=NamedRerouteUsage())
  )
  NamedRerouteDeclaration(None=Step(Y=Max(A=@n138, B=@n139), X=Multiply(A=@n140, B=@n142))) {name=Mobile Depth Lines, nodeColor=<rgba 0.0324 1 0 1>}
  n143 = Subtract(A=@n66, B=ScalarParameter() {group=Outlines Roughness, parameterName=RoughnessThreshold Start}) {constB=20000}
  n144 = ScalarParameter() {defaultValue=6000, group=Outlines Roughness, parameterName=Roughness Theshold Falloff}
  NamedRerouteDeclaration(None=Saturate(None=Divide(A=@n143, B=@n144) {constB=2000})) {name=Roughness Soft Blend Mask, nodeColor=<rgba 0 1 0.4924 1>}
  n145 = SceneTexelSize()
  n146 = ComponentMask(None=@n145) {r=true}
  n147 = ComponentMask(None=@n145) {g=true}
  n148 = AppendVector(A=@n146, B=@n147)
  n149 = AppendVector(A=@n114, B=Multiply(A=@n116) {constB=-1})
  NamedRerouteDeclaration(None=LinearInterpolate(A=Subtract(A=@n112, B=@n148), B=Add(A=@n113, B=@n149) {desc=y+}, Alpha=NamedRerouteUsage())) {name=UV2, nodeColor=<rgba 0 0.6514 1 1>}
  n150 = Constant()
  n151 = AppendVector(A=@n150, B=@n147)
  n152 = ComponentMask(None=@n115) {r=true}
  n153 = AppendVector(A=@n152, B=@n114)
  NamedRerouteDeclaration(None=LinearInterpolate(A=Subtract(A=@n112, B=@n151), B=Add(A=@n113, B=@n153) {desc=x-}, Alpha=NamedRerouteUsage())) {name=UV3, nodeColor=<rgba 0.5 0 1 1>}
  n154 = AppendVector(A=@n146, B=@n150)
  n155 = AppendVector(A=Multiply(A=@n152) {constB=-1}, B=@n114)
  NamedRerouteDeclaration(None=LinearInterpolate(A=Subtract(A=@n112, B=@n154), B=Add(A=@n113, B=@n155) {desc=y-}, Alpha=NamedRerouteUsage())) {name=UV4, nodeColor=<rgba 0.6641 1 0 1>}
  n156 = CameraVectorWS()
  n157 = Max(A=DotProduct(A=@n156, B=NamedRerouteUsage()), B=DotProduct(A=@n156, B=NamedRerouteUsage()))
  n158 = Max(A=DotProduct(A=@n156, B=NamedRerouteUsage()), B=DotProduct(A=@n156, B=NamedRerouteUsage()))
  NamedRerouteDeclaration(None=Step(
    Y=Max(A=@n157, B=@n158)
    X=ScalarParameter() {defaultValue=0.1, group=Outlines Depth, parameterName=DepthFresnel, sliderMax=30, sortPriority=1}
   ) {constX=0.05}) {name=Fresnel, nodeColor=<rgba 1 0 0.1172 1>}
  MaterialExpressionNamedRerouteUsage_62 = NamedRerouteUsage()
  MaterialExpressionNamedRerouteUsage_61 = NamedRerouteUsage()
  MaterialExpressionNamedRerouteUsage_60 = NamedRerouteUsage()
  MaterialExpressionNamedRerouteUsage_63 = NamedRerouteUsage()
  MaterialExpressionNamedRerouteUsage_64 = NamedRerouteUsage()
  MaterialExpressionNamedRerouteUsage_56 = NamedRerouteUsage()
  MaterialExpressionNamedRerouteUsage_65 = NamedRerouteUsage()
  n159 = Custom!(
   Tex=@MaterialExpressionNamedRerouteUsage_62
   UV=@MaterialExpressionNamedRerouteUsage_61
   TraceVec=@MaterialExpressionNamedRerouteUsage_60
   NumSteps=@MaterialExpressionNamedRerouteUsage_63
   bias=@MaterialExpressionNamedRerouteUsage_64
   StepSizeMAX=@MaterialExpressionNamedRerouteUsage_56
   MaxDepth=@MaterialExpressionNamedRerouteUsage_65
  ) {
   code="// The below expression will get compiled\r\n// into the output of this node\r\n//float3(1, 1, 1);\r\n//float3 Color = SceneTextureLookup(ViewportUVToSceneTextureUV(UV, 1), 1, false);\r\n\r\nfloat TimeLerp = 1;\r\nfloat DepthDiff = 0;\r\n//float DepthDiff2 = 0;\r\n\r\nfloat LastDiff = -bias;\r\nfloat StepSize = StepSizeMAX/NumSteps;\r\nDebug = 1;\r\nfloat3 RayStepUVz = float3(TraceVec.x, TraceVec.y, TraceVec.z);\r\n\r\nfloat accum = 0;\r\nfloat3 RayUVz = float3(UV, 0 );\r\n//float3 RayUVzInv = float3(UV, 0 );\r\nfloat depth = SceneTextureLookupCommon(Parameters, ViewportUVToSceneTextureUV(UV, 1), 1, false);\r\nint i = 0;\r\nwhile (i < NumSteps)\r\n{\r\n\r\n\tRayUVz += RayStepUVz*StepSize;\r\n    //RayUVzInv -= RayStepUVz*StepSize;\r\n    RayUVz.xy = saturate(RayUVz.xy);\r\n    //RayUVzInv.xy = saturate(RayUVzInv.xy);\r\n\r\n    float MaskSample1 = SceneTextureLookupCommon(Parameters, ViewportUVToSceneTextureUV(RayUVz.xy, 23), 23, false);\r\n    if (MaskSample1 == 4)\r\n\t\t{\r\n\r\n        MaskSample1 = 1;\r\n        }\r\n    else\r\n\t\t{\r\n\r\n        MaskSample1 = 0;\r\n        i++;\r\n        continue;\r\n\r\n        }  \r\n   /* float MaskSample2 = SceneTextureLookupCommon(Parameters, ViewportUVToSceneTextureUV(RayUVzInv.xy, 23), 23, false);\r\n    if (MaskSample2 == 4)\r\n\t\t{\r\n\r\n        MaskSample2 = 1;\r\n\r\n        }\r\n    else\r\n\t\t{\r\n\r\n        MaskSample2 = 0;\r\n\r\n        }  \r\n*/ \tfloat SampleDepth = SceneTextureLookupCommon(Parameters, ViewportUVToSceneTextureUV(RayUVz.xy, 1), 1, false);\r\n    //float SampleDepth2 = SceneTextureLookupCommon(Parameters, ViewportUVToSceneTextureUV(RayUVzInv.xy, 1), 1, false);\r\n\t//DepthDiff = (SampleDepth+RayUVz.z)-depth + max((SampleDepth2-depth),0);\r\n   // DepthDiff = (SampleDepth+RayUVz.z)-depth + (SampleDepth2-depth);\r\n    DepthDiff = (SampleDepth+RayUVz.z)-depth ;\r\n    DepthDiff *= MaskSample1; \r\n    //DepthDiff = (SampleDepth+RayUVz.z)-depth;\r\n    //DepthDiff = (SampleDepth+RayUVz.z)-depth;\r\n    //DepthDiff2 = (SampleDepth2+RayUVzInv.z)-depth;\r\n    //DepthDiff2 *= MaskSample2; \r\n    \r\n\r\n\t\tif (DepthDiff < 0 &&&& DepthDiff > -MaxDepth)\r\n\t\t{\r\n\r\n\t\t\t//if(LastDiff > 0)\r\n\t\t\t//{\r\n\t\t\t//\tTimeLerp = saturate( LastDiff / (LastDiff - DepthDiff));\r\n\t\t\t//\taccum += StepSize * (1-TimeLerp);\r\n\t\t\t\t//accum+=StepSize;\r\n\t\t\t//}\r\n\t\t\t//else\r\n\t\t\t//{\r\n\t\t\t//\taccum+=StepSize;\r\n\t\t\t//}\r\n            accum += 0.1;\r\n\r\n\t\t}\r\n\t\telse\r\n\t\t//if(LastDiff <= 0)\r\n\t\t{\r\n\t\t\t//accum -=0.1;\r\n\t\t}\r\n //       if (DepthDiff2 < 0 )\r\n\t//\t{\r\n   //        accum +=0.1;\r\n        \r\n    //    }\r\n\r\n\t//LastDiff = DepthDiff;\r\n\r\n\ti++;\r\n}\r\n//Debug =  SceneTextureLookup(GetDefaultSceneTextureUV(Parameters, 23), 23, false);\r\nreturn 1-saturate(accum);\r\n"
   desc=no custom depth masking of clear coat objects
   additionalOutputs [{outputName=Debug, outputType=CMOT_Float1}]
   inputs [
    {inputName=Tex, input {expression=&self:MaterialExpressionNamedRerouteUsage_62}}
    {inputName=UV, input {expression=&self:MaterialExpressionNamedRerouteUsage_61}}
    {inputName=TraceVec, input {expression=&self:MaterialExpressionNamedRerouteUsage_60}}
    {inputName=NumSteps, input {expression=&self:MaterialExpressionNamedRerouteUsage_63}}
    {inputName=bias, input {expression=&self:MaterialExpressionNamedRerouteUsage_64}}
    {inputName=StepSizeMAX, input {expression=&self:MaterialExpressionNamedRerouteUsage_56}}
    {inputName=MaxDepth, input {expression=&self:MaterialExpressionNamedRerouteUsage_65}}
   ]
  }
  NamedRerouteDeclaration(None=@n159.Debug) {name=Debug Value, nodeColor=<rgba 0 0.7656 1 1>}
  NamedRerouteUsage()
  OneMinus(None=@n123)
  OneMinus(None=NamedRerouteUsage())
  MaterialExpressionNamedRerouteUsage_74 = NamedRerouteUsage()
  MaterialExpressionNamedRerouteUsage_75 = NamedRerouteUsage()
  n160 = Custom!(UV=@MaterialExpressionNamedRerouteUsage_74, Tex=@MaterialExpressionNamedRerouteUsage_75) {
   code="float MaskSample1 = SceneTextureLookupCommon(Parameters, ViewportUVToSceneTextureUV(UV, 13), 13, false);\r\nreturn MaskSample1;"
   showCode=true
   inputs [
    {inputName=UV, input {expression=&self:MaterialExpressionNamedRerouteUsage_74}}
    {inputName=Tex, input {expression=&self:MaterialExpressionNamedRerouteUsage_75}}
   ]
  }
  n161 = NamedRerouteUsage()
  OneMinus(None=Saturate(None=Subtract(A=@n160, B=@n161)))
  Reroute(None=@n68)
  Saturate(None=Divide(A=ComponentMask(None=@n160) {r=true}) {constB=5000})
  SceneTexture() {sceneTextureId=PPI_CustomDepth}
  n162 = LinearInterpolate(A=Multiply(A=NamedRerouteUsage(), B=NamedRerouteUsage()), B=@n6, Alpha=NamedRerouteUsage())
  n163 = NamedRerouteUsage()
  ShadingPathSwitch(
   Default=QualitySwitch(Default=@n162, Low=@n162, High=@n162, Medium=@n162, Epic=Multiply(A=@n162, B=@n163))
   Mobile=StaticSwitchParameter(True=NamedRerouteUsage(), False=@n6) {defaultValue=true, parameterName=UsePosturizerForMobile}
  )
 ]
}
